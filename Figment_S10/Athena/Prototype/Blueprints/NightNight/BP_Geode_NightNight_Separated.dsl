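; Geode actor for the NightNight prototype that plays a pass-by sound on a separated missile silo stage when listeners are nearby.
; On BeginPlay, where cosmetic events can execute, it enables Tick after a short delay.
; Tick checks for listeners within a large radius, then pauses itself, attaches the cue to MissileSilo_Stage2_Main and re-enables after a cooldown.
; The DisablePassbyAudio flag switches Tick off instead.
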(event EventBeginPlay
  (Utilities|FlowControl|Sequence
    (:then_1
      (Networking|CanExecuteCosmeticEvents
        (:True
          (Utilities|FlowControl|Delay 3.0)
          (Actor|Tick|SetActorTickEnabled true))))))

(event EventTick
  (params (DeltaSeconds :type Float%20%28single-precision%29))
  (if (Variables|Default|GetDisablePassbyAudio)
    (Actor|Tick|SetActorTickEnabled false)
    (else
      (Utilities|FlowControl|Sequence
        (:then_1
          (bind _returnvalue (Audio|AreAnyListenersWithinRange (Transformation|GetActorLocation) 5e+04))
          (if _returnvalue
            (Actor|Tick|SetActorTickEnabled false)
            (Audio|SpawnSoundAttached "/Figment_S10/Athena/Sounds/Events/NightNight/NightNight_Geode_PassBy_Dynamic_01_Cue.NightNight_Geode_PassBy_Dynamic_01_Cue" (Variables|Default|GetMissileSilo_Stage2_Main))
            (Utilities|FlowControl|Delay 1e+01)
            (Actor|Tick|SetActorTickEnabled true)))))))

(fn ConstructionScript)
