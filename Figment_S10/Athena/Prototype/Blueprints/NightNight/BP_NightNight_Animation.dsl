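; NightNight animation actor that drives a cinematic camera along a spline toward a final location.
; Clients snap it to the local pawn or spectated view target, destroying it if neither resolves.
; The server starts the movement timeline, whose remote-only updates spawn the camera once and move it along the spline, facing the destination.
; SetViewTarget, uncalled here, sets spline endpoints and blends the player's view to that camera.
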(event EventBeginPlay
  (bind _returnvalue (Fortnite|GetLocalFortPlayerControllers true))
  (bind _output (Utilities|Array|Get(acopy) _returnvalue))
  (bind _pawn (Variables|Default|GetPawn))
  (Networking|SwitchHasAuthority
    (:Authority
      (CallFunction|StartServerTimeline))
    (:Remote
      (Utilities|IsValid _output
        (:Is%20Valid
          (bind _output_get (Variables|Default|SetLocalController _output))
          (Utilities|IsValid (Game|Player|GetPlayerPawn _output_get)
            (:Is%20Valid
              (Variables|Default|SetPawn (Game|Player|GetPlayerPawn _output_get)))
            (:Is%20Not%20Valid
              ;[1 1 1] If dead and spectating
              (bind _asfort_player_pawn_athena (Utilities|Casting|CastToFortPlayerPawnAthena (Pawn|GetViewTarget (Variables|Default|GetLocalController)))
                (:then
                  (Variables|Default|SetPawn _asfort_player_pawn_athena))
                (:CastFailed
                  (Actor|DestroyActor)))))
          (bind _mesh (Variables|/Script/Engine.Character|GetMesh _pawn)
            (:then
              (Variables|Default|SetPlayerMesh _mesh)
              (bind _output_get_2 (Variables|Default|SetPlayerLocation (Transformation|GetActorLocation _pawn)))
              (bind _output_get_3 (Variables|Default|SetPlayerRotation (Transformation|GetActorRotation _pawn)))
              (Transformation|SetActorLocationAndRotation _output_get_2 _output_get_3))
            (:else)))
        (:Is%20Not%20Valid
          (Actor|DestroyActor))))))

(event Custom|SpawnCamera
  (bind _returnvalue (Game|SpawnActorfromClass "/Figment_S10/Athena/Prototype/Blueprints/NightNight/NightNight_Camera.NightNight_Camera_C" (Math|Transform|MakeTransform (Variables|Default|GetPlayerLocation) (Variables|Default|GetPlayerRotation)) "AlwaysSpawn" "MultiplyWithRoot" 0 0))
  (Variables|Default|SetNightNightCamera _returnvalue))

(event Custom|SetViewTarget
  (Spline|SetLocationatSplinePoint (Variables|Default|GetMovementSpline) 0 (Transformation|GetActorLocation (Pawn|GetViewTarget (Variables|Default|GetLocalController))) "World" false)
  (Spline|SetLocationatSplinePoint (Variables|Default|GetMovementSpline) 1 (Variables|Default|GetFinalLocation) "World" false)
  (Game|Player|SetViewTargetwithBlend (Variables|Default|GetLocalController) (Variables|Default|GetNightNightCamera) 5.0 "VTBlend_Linear" 1.0 true))

(event Custom|StartServerTimeline
  (bind _returnvalue_location (Math|Transform|BreakTransform (Spline|GetTransformatTime (Variables|Default|GetMovementSpline) (Variables|Default|GetSplineTime) "World" true false)))
  (bind (_direction _movealpha _fullmovecurve) (Timeline|ServerMovementTimeline PlayFromStart)
    (:Update
      (Networking|SwitchHasAuthority
        (:Remote
          (Utilities|FlowControl|Sequence
            (:then_0
              (Utilities|FlowControl|DoOnce false
                (:Completed
                  (CallFunction|SpawnCamera))))
            (:then_1
              (Variables|Default|SetSplineTime _fullmovecurve)
              (Transformation|SetActorLocationAndRotation (Variables|Default|GetNightNightCamera) _returnvalue_location (Math|Rotator|FindLookatRotation _returnvalue_location (Variables|Default|GetFinalLocation)) false false))))))))

(event Collision|EventActorBeginOverlap
  (params (OtherActor :type Actor%20Object%20Reference)))

(event EventTick
  (params (DeltaSeconds :type Float%20%28single-precision%29)))

(fn ConstructionScript)
