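; NightNight prototype actor that plays local camera shake and controller force feedback through Small, Medium, Large and ExtendedLength events, skipped on dedicated servers.
; Large plays at full scale; the others use CameraShakeScaleFromDistance, a power-curve falloff on view-target distance from this actor, clamped and debug-printed.
; Small, Medium and Large share the large shake class, differing in scale and feedback asset.
; ExtendedLength uses a rocket lift-off shake with large feedback.
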(event Custom|SmallShake
  (if (not (Networking|IsDedicatedServer))
    (bind (_localplayercontroller _scale) (CallFunction|CameraShakeScalefromDistance (* (Variables|Default|GetRangeMin) (Variables|Default|GetRangeMultiSmall))))
    (Game|Feedback|ClientStartCameraShake :self _localplayercontroller :Shake "/Figment_S10/Athena/Prototype/Blueprints/NightNight/CameraShakeFF/NightNight_CameraShake_Large.NightNight_CameraShake_Large_C" :Scale _scale :PlaySpace "CameraLocal")
    (Game|Feedback|ClientPlayForceFeedback _localplayercontroller "/Figment_S10/Athena/Prototype/Blueprints/NightNight/CameraShakeFF/NightNight_ForceFeedback.NightNight_ForceFeedback" "None" false false false)))

(event Custom|MediumShake
  (if (not (Networking|IsDedicatedServer))
    (bind (_localplayercontroller _scale) (CallFunction|CameraShakeScalefromDistance (* (Variables|Default|GetRangeMin) (Variables|Default|GetRangeMultiMed))))
    (Game|Feedback|ClientStartCameraShake :self _localplayercontroller :Shake "/Figment_S10/Athena/Prototype/Blueprints/NightNight/CameraShakeFF/NightNight_CameraShake_Large.NightNight_CameraShake_Large_C" :Scale _scale :PlaySpace "CameraLocal")
    (Game|Feedback|ClientPlayForceFeedback _localplayercontroller "/Figment_S10/Athena/Prototype/Blueprints/NightNight/CameraShakeFF/NightNight_ForceFeedback_Medium.NightNight_ForceFeedback_Medium" "None" false false false)))

(event Custom|LargeShake
  (if (not (Networking|IsDedicatedServer))
    (bind _output_get (Variables|Default|SetLocalPlayerController (Utilities|Array|Get(acopy) (Fortnite|GetLocalFortPlayerControllers true))))
    (Game|Feedback|ClientStartCameraShake :self _output_get :Shake "/Figment_S10/Athena/Prototype/Blueprints/NightNight/CameraShakeFF/NightNight_CameraShake_Large.NightNight_CameraShake_Large_C" :Scale 1.0 :PlaySpace "CameraLocal")
    (Game|Feedback|ClientPlayForceFeedback _output_get "/Figment_S10/Athena/Prototype/Blueprints/NightNight/CameraShakeFF/NightNight_ForceFeedback_Large.NightNight_ForceFeedback_Large" "None" false false false)))

(event Custom|ExtendedLengthShake
  (if (not (Networking|IsDedicatedServer))
    (bind _output_get (Variables|Default|SetLocalPlayerController (Utilities|Array|Get(acopy) (Fortnite|GetLocalFortPlayerControllers true))))
    (bind (_localplayercontroller _scale) (CallFunction|CameraShakeScalefromDistance (* (Variables|Default|GetRangeMin) (Variables|Default|GetRangeMultiSmall))))
    (Game|Feedback|ClientStartCameraShake :self _output_get :Shake "/Figment_S10/Athena/Prototype/Blueprints/NightNight/CameraShakeFF/NightNight_LiftOffRocket_CameraShake.NightNight_LiftOffRocket_CameraShake_C" :Scale _scale :PlaySpace "CameraLocal")
    (Game|Feedback|ClientPlayForceFeedback _localplayercontroller "/Figment_S10/Athena/Prototype/Blueprints/NightNight/CameraShakeFF/NightNight_ForceFeedback_Large.NightNight_ForceFeedback_Large" "None" false false false)))

(fn CameraShakeScaleFromDistance
  (params (RangeMin :type Float%20%28double-precision%29))
  (bind _localplayercontroller (Variables|Default|GetLocalPlayerController))
  (bind _returnvalue (Pawn|GetViewTarget _localplayercontroller))
  (bind _camerashakepower (Variables|Default|GetCameraShakePower))
  (bind _returnvalue_2 (Transformation|GetActorLocation _returnvalue))
  (bind _returnvalue_3 (Transformation|GetActorLocation))
  (bind _returnvalue_4 (Math|Vector|Distance(Vector) _returnvalue_3 _returnvalue_2))
  (bind _returnvalue_5 (Math|Float|MapRangeClamped _returnvalue_4 0.0 RangeMin 1.0))
  (bind _returnvalue_6 (Math|Float|Power _returnvalue_5 _camerashakepower))
  (bind _returnvalue_7 (Math|Float|Clamp(Float) _returnvalue_6 0.3))
  (Variables|Default|SetLocalPlayerController (Utilities|Array|Get(acopy) (Fortnite|GetLocalFortPlayerControllers true)))
  (Development|PrintString (Utilities|String|ToString(Float) _returnvalue_7))
  (return (Variables|Default|GetLocalPlayerController) _returnvalue_7))

(fn ConstructionScript)
