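; NightNight prototype volume that clears out damageable BuildingActors that overlap it.
; On BeginPlay only the server enables overlap collision on the cube, while clients switch collision off.
; Overlapped actors receive a short random lifespan, which staggers their removal.
; A separate timer-looped DetroyProps event drains a prop array with heavy environment damage, but nothing in this graph calls it initially or fills the array.
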(event EventBeginPlay
  (bind _cube (Variables|Default|GetCube))
  (if (Networking|IsServer)
    (Collision|SetCollisionProfileName _cube "OverlapAllDynamic" true :!class /Script/Engine.PrimitiveComponent)
    (Collision|SetActorEnableCollision true)
    (else
      (Collision|SetCollisionProfileName _cube "nocollision" true :!class /Script/Engine.PrimitiveComponent)
      (Collision|SetActorEnableCollision false))))

(event Collision|EventActorBeginOverlap
  (params (OtherActor :type Actor%20Object%20Reference))
  (bind _asbuilding_actor (Utilities|Casting|CastToBuildingActor OtherActor))
  (Utilities|IsValid _asbuilding_actor
    (:Is%20Valid
      (if (Variables|/Script/Engine.Actor|GetbCanBeDamaged _asbuilding_actor)
        (Actor|SetLifeSpan _asbuilding_actor (Math|Random|RandomFloatinRange 0.1 0.3))))))

(event EventTick
  (params (DeltaSeconds :type Float%20%28single-precision%29)))

(event Custom|DetroyProps
  (if (Math|Conversions|ToBoolean(Integer) (Utilities|Array|Length (Variables|Default|GetPropsToDestroy)))
    (Variables|Default|SetTargetProp (Utilities|Array|Get(acopy) (Variables|Default|GetPropsToDestroy)))
    (Utilities|IsValid (Variables|Default|GetTargetProp)
      (:Is%20Valid
        (Environment|ApplyGlobalEnvironmentDamagetoActor (Variables|Default|GetTargetProp) 2e+03))
      (:Is%20Not%20Valid
        _))
    (Utilities|Array|RemoveItem (Variables|Default|GetPropsToDestroy) (Variables|Default|GetTargetProp))
    (bind _returnvalue (Utilities|Time|SetTimerbyEvent (EventDispatchers|CreateEvent :!event Custom|DetroyProps) 0.04))
    (Variables|Default|SetPropDestroyTimer _returnvalue)
    (else
      (Utilities|Time|ClearandInvalidateTimerbyHandle (Variables|Default|GetPropDestroyTimer)))))

(fn ConstructionScript)
