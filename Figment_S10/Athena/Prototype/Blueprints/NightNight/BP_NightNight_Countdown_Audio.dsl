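; Countdown audio actor for the NightNight event that plays klaxons and launch warning audio around a watched calendar event.
; Once the match is ready, the event's remaining time drives timers for either hourly klaxons or per-minute warnings.
; Timers increment a replicated integer whose OnRep plays the klaxon, syncing clients without RPCs.
; At event end, PlayingAudio toggles on, then off after a final timer, with OnRep_StartAudio handling fades and the final klaxon.
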(event Custom|MinuteWarnings
  (if (>= (Variables|Default|GetCurrentSeconds) 1.0)
    (Math|Float|DecrementFloat (Variables|Default|GetCurrentSeconds)
      (:%20%20
        (Variables|Default|SetKlaxonEventInt (+ (Variables|Default|GetKlaxonEventInt) 1))
        (Utilities|Time|SetTimerbyEvent (delegate Custom|MinuteWarnings) 6e+01)))))

(event Custom|Play
  (Variables|Default|SetKlaxonEventInt (+ (Variables|Default|GetKlaxonEventInt) 1)))

(event Custom|TotalTime
  (Variables|Default|SetCurrentSeconds (Math|Conversions|ToFloat(Integer) (Math|Float|Floor (/ (Variables|Default|GetCurrentSeconds) (Variables|Default|GetMinuteWarning)))))
  (CallFunction|MinuteWarnings))

(event Custom|SetKlaxon
  (Variables|Default|SetKlaxonEventInt (+ (Variables|Default|GetKlaxonEventInt) 1))
  (Utilities|Time|SetTimerbyEvent (delegate Custom|Play) 3.6e+03 true))

(event EventBeginPlay
  (Class|FortAsyncActionAuthorityMatchReady|MatchReadyAsync
    (:OnReady
      (bind (_eventname _timeuntilend) (MCPContext|WatchCalendarEvent (Variables|Default|GetPre-Launch%20Klaxons))
        (:OnEventActive
          (bind _output_get (Variables|Default|SetTime%20Until%20Countdown%20End _timeuntilend))
          (bind _output_get_2 (Variables|Default|SetCurrentSeconds (Math|Float|float+float (* (.Minutes (Math|Timespan|BreakTimespan _output_get)) 6e+01) (Math|Conversions|ToFloat(Integer) (.Seconds (Math|Timespan|BreakTimespan _output_get))) (* (.Milliseconds (Math|Timespan|BreakTimespan _output_get)) 0.001))))
          (bind _output_get_3 (Variables|Default|SetSecondsUnderTenMinutes (+ (Math|Conversions|ToFloat(Integer) (.Seconds (Math|Timespan|BreakTimespan _output_get))) (* (.Milliseconds (Math|Timespan|BreakTimespan _output_get)) 0.001))))
          (if (<= _output_get_3 (Variables|Default|GetTenMinuteWarning))
            (CallFunction|ContinueCountdownEvent)
            (else
              (Utilities|Time|SetTimerbyEvent (delegate Custom|SetKlaxon) _output_get_2))))
        (:OnEventEnded
          (Variables|Default|SetPlayingAudio true)
          (Utilities|Time|SetTimerbyEvent (delegate Custom|TurnOffAudio) (Variables|Default|GetFinalTimer)))))))

(event Custom|Continue%20CountdownEvent
  (Utilities|Time|SetTimerbyEvent (delegate Custom|TotalTime) (Variables|Default|GetSecondsUnderTenMinutes)))

(event Custom|TurnOffAudio
  (Variables|Default|SetPlayingAudio false))

(fn EventPlayKlaxonCue)

(fn OnRep_KlaxonEventInt
  (Audio|Components|Audio|Play (Variables|Default|GetKlaxonAudio) 0.0))

(fn OnRep_StartAudio
  (bind _launchwarningaudio (Variables|Default|GetLaunchWarningAudio))
  (bind _klaxonaudio (Variables|Default|GetKlaxonAudio))
  (if (Variables|Default|GetPlayingAudio)
    (Audio|Components|Audio|FadeIn _launchwarningaudio 1.0 1.0 0.0 "Linear" :!class /Script/Engine.AudioComponent)
    (Audio|Components|Audio|SetSound _klaxonaudio "/Figment_S10/Athena/Sounds/Events/NightNight/NightNight_Klaxon_Final_Cue.NightNight_Klaxon_Final_Cue")
    (Audio|Components|Audio|Play _klaxonaudio 0.0)
    (else
      (Audio|Components|Audio|FadeOut _launchwarningaudio 1.0 0.0 "Linear" :!class /Script/Engine.AudioComponent)
      (Audio|Components|Audio|Play (Variables|Default|GetLaunchWarningStartAudio) 0.0))))

(fn SetSoundCue
  (params (NewSound :type Sound%20Base%20Object%20Reference)))

(fn StartLaunchWarningAudio
  (Audio|Components|Audio|FadeIn (Variables|Default|GetLaunchWarningAudio) 1.0 1.0 0.0 "Linear" :!class /Script/Engine.AudioComponent))

(fn ConstructionScript)
