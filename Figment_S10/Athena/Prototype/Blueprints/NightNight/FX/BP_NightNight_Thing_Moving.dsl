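; Moving NightNight prototype effect actor that launches nearby players when its KnockPlayers event runs.
; KnockPlayers stores the MeteorGlitch component's world location and sphere-overlaps Fortnite player pawns within the Radius variable.
; Each pawn is launched with a large fixed strength along its look-at direction toward that location, despite the event's name.
; MeteorGlitch collision is then disabled; the overlap and tick events are empty.
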(event Custom|KnockPlayers
  (bind _meteorglitch (Variables|Default|GetMeteorGlitch))
  (bind _returnvalue (Transformation|GetWorldLocation _meteorglitch))
  (Variables|Default|SetThisLoc _returnvalue)
  (bind _outactors (Collision|SphereOverlapActors _returnvalue (Variables|Default|GetRadius) (Utilities|Array|MakeArray "ObjectTypeQuery3") "/Script/FortniteGame.FortPlayerPawn"))
  (for _array_element _outactors
    (bind _asfort_player_pawn (Utilities|Casting|CastToFortPlayerPawn _array_element))
    (Utilities|LaunchPlayerPawn _asfort_player_pawn _asfort_player_pawn (* (Math|Vector|GetForwardVector (Math|Rotator|FindLookatRotation (Transformation|GetActorLocation _asfort_player_pawn) (Variables|Default|GetThisLoc))) 3.6e+04)))
  (Collision|SetCollisionEnabled (Variables|Default|GetMeteorGlitch) "NoCollision"))

(event Collision|EventActorBeginOverlap
  (params (OtherActor :type Actor%20Object%20Reference)))

(event EventTick
  (params (DeltaSeconds :type Float%20%28single-precision%29)))

(fn ConstructionScript
  (|CallParentFunction))
