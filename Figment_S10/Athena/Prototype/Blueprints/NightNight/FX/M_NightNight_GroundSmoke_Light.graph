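# Masked unlit mesh-particle material for NightNight ground smoke, adding a colored point-light glow to stylized shading.
# World-aligned cloud normals scroll vertically and blend toward vertex normals, driving a sharpened fake directional light and Fresnel/top rim that mix dark and bright colors.
# A sphere mask above the actor adds the glow, with radius, height and intensity set by dynamic particle parameters.
# Particle alpha height-lerps against view-facing normals, multiplied by camera depth fade, to form the cutout.
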Material /Figment_S10/Athena/Prototype/Blueprints/NightNight/FX/M_NightNight_GroundSmoke_Light
graph {
 properties {
  bTangentSpaceNormal=false
  bUsedWithMeshParticles=true
  blendMode=BLEND_Masked
  opacityMaskClipValue=0.9
  pixelDepthOffsetMode=PDOM_Legacy
  shadingModel=MSM_Unlit
  usedShadingModels=Unlit
 }
 nodes [
  n1 = Constant2Vector()
  n2 = DynamicParameter() {defaultValue=<rgba 0 0 0 1>, paramNames [TextureTile, LightPositionRandom, LightHightAdd, LightIntensity]}
  n3 = Multiply(A=@n2.LightHightAdd) {constB=1000}
  n4 = SphereMask(
   A=WorldPosition().XYZ
   B=Add(A=ActorPositionWS(), B=AppendVector(A=@n1, B=@n3))
   Radius=Multiply(A=@n2.LightPositionRandom, B=ScalarParameter() {defaultValue=800, parameterName=Light_Radius})
   Hardness=ScalarParameter() {parameterName=Light_Hardness}
  )
  n5 = OneMinus(None=Power(Base=OneMinus(None=@n4)) {constExponent=3})
  n6 = VectorParameter() {parameterName=LightColor, defaultValue=<rgba 1 0.7492 0.1374 1>}
  n7 = Constant3Vector() {constant=<rgba 500 500 500 0>}
  n8 = DynamicParameter() {defaultValue=<rgba 0 0 0 1>, paramNames [TextureTile, LightPositionRandom, LightHightAdd, LightIntensity]}
  n9 = Multiply(A=Time(), B=Multiply(A=Constant() {r=-0.05}, B=Constant3Vector() {constant=<rgba 0 0 300 0>})) {constB=200}
  n10 = WorldPosition()
  n11 = LinearInterpolate(
   A=MaterialFunctionCall!(
    Diffuse Texture Object=TextureObject() {samplerType=SAMPLERTYPE_LinearColor, texture=/Figment_S10/Effects/Fort_Effects/Effects/Explosion/Cloud_Diffuse}
    Spec Texture Object
    TextureSize=Multiply(A=@n7, B=@n8.TextureTile) {constB=3}
    --------------------------
    Use seperate settings for Z projection
    Use seperate texture for Z projection
    Texture Size Z projection
    Z Plane Diffuse Texture Object
    Z Plane Spec Texture Object
    "--------------------------      "
    ProjectionTransitionContrast
    Use Texture Alpha Heightmap to Refine the Transistion
    HeightLerpTransitionContrast
    PickTheDiffuseHeightChannel
    TheDiffuseHeightChannel
    "--------------------------       "
    NormalTextureObject=TextureObject() {samplerType=SAMPLERTYPE_Normal, texture=/Figment_S10/Effects/Fort_Effects/Effects/Explosion/Cloud_normals}
    World Space Normals
    "--------------------------  "
    Export Float 4 Diffuse
    Export Float 4 Spec
    "--------------------------     "
    WorldPosition=Add(A=@n9, B=@n10.XYZ)
    World Space Normal
   ) {materialFunction=/Engine/Functions/Engine_MaterialFunctions02/Texturing/WorldAlignedTextures_Complex}.Normal XYZ Texture
   B=VertexNormalWS()
   Alpha=Constant() {desc=.2, r=0.2}
  ) {desc=Soft normal}
  n12 = DotProduct(A=@n11, B=Transform(None=Constant3Vector() {desc=Angle of light, constant=<rgba 0.2 1.2 0 0>}) {transformSourceType=TRANSFORMSOURCE_View})
  n13 = Clamp(None=Multiply(A=Add(A=@n12) {constB=-0.1}) {constB=40})
  n14 = Constant() {r=1}
  n15 = Multiply(A=@n14) {constB=3}
  n16 = Add(A=Multiply(A=Fresnel()) {constB=10}, B=DotProduct(A=@n11, B=Constant3Vector() {constant=<rgba 0 0 1 0>}))
  n17 = Clamp(None=Multiply(A=Add(A=@n16) {constB=-0.4, desc="1-0 (line)"}) {constB=15, desc=sharp line})
  n18 = Add(A=Multiply(A=Multiply(A=@n13, B=@n15) {constB=0.5}) {constB=0.2}, B=@n17)
  Add(
   A=Multiply(A=Multiply(A=@n5, B=@n6.RGB), B=@n2.LightIntensity)
   B=LinearInterpolate(
    A=VectorParameter() {parameterName=DarkColor, defaultValue=<rgba 0.076 0.08 0.12 1>}.RGB
    B=VectorParameter() {parameterName=BirghtColor, defaultValue=<rgba 0.09 0.1013 0.135 1>}.RGB
    Alpha=Power(Base=@n18) {constExponent=3}
   )
  )
  Constant() {r=1}
  n19 = MaterialFunctionCall!(DayBloom_Mult=Constant() {r=1}, In=@n14) {materialFunction=/Game/Packages/Fortnite_UniqueMaterials_1/MH_Materials/DayNight_Auto_Bloom_Control}
  LinearInterpolate(A=@n14, B=Multiply(A=Multiply(A=@n19.Result) {constB=2}), Alpha=@n17)
  ObjectPositionWS()
  n20 = ParticleColor()
  n21 = DotProduct(A=@n11, B=CameraVectorWS())
  n22 = DotProduct(A=VertexNormalWS(), B=CameraVectorWS())
  n23 = Divide(A=Add(A=Clamp(None=@n21), B=Clamp(None=@n22)))
  n24 = Constant() {r=2000}
  Reroute(None=Multiply(
    A=MaterialFunctionCall(Transition Phase=@n20.A, Height Texture=@n23) {materialFunction=&HeightLerp}.Alpha
    B=MaterialFunctionCall(Fade Length=@n24) {materialFunction=&CameraDepthFade}.Result
   ))
 ]
}
