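# Additive, unlit, two-sided mesh-particle material for the NightNight meteor shell, rendering scrolling smoky noise tinted by particle color.
# Two tiled, panning noise layers, driven by time plus a dynamic PanAdd offset, are sampled twice at slightly shifted UVs and tinted pink and green for chromatic fringing.
# The blend is raised to a power and saturation-boosted through negative desaturation.
# A vertical UV mask, view-angle fade and particle alpha soften the shell.
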Material /Figment_S10/Athena/Prototype/Blueprints/NightNight/FX/M_NightNight_Meteor_Shell
graph {
 properties {
  bUsedWithMeshParticles=true
  blendMode=BLEND_Additive
  pixelDepthOffsetMode=PDOM_Legacy
  previewMesh=&/SM_NightNight_RocketShell
  shadingModel=MSM_Unlit
  twoSided=true
  usedShadingModels=Unlit
 }
 nodes [
  n1 = ParticleColor()
  n2 = Constant() {r=0.5}
  n3 = SphereMask(A=ComponentMask(None=TextureCoordinate()) {g=true}, B=@n2, Radius=@n2) {hardnessPercent=70}
  n4 = MaterialFunctionCall("FadeSensitivity (optional)"=ScalarParameter() {defaultValue=1, parameterName=AngleFade}) {materialFunction=&FadeBasedOnViewAngle}
  n5 = Multiply(A=Multiply(A=@n4.Result, B=@n4.Result)) {constB=2}
  Multiply(A=@n1.A, B=Multiply(A=Power(Base=@n3), B=Saturate(None=@n5)))
  n6 = ComponentMask(None=VectorParameter() {parameterName=Tile1, defaultValue=<rgba 3 1.4 0 0>}.RGB) {g=true, r=true}
  n7 = Add(
   A=DynamicParameter() {paramNames [PanAdd,Param2,Param3,Param4]}.PanAdd
   B=Multiply(A=Time(), B=ScalarParameter() {defaultValue=1, parameterName=PanSpeed})
  )
  n8 = Panner(
   Coordinate=Multiply(A=TextureCoordinate(), B=@n6)
   Time=@n7
   Speed=ComponentMask(None=VectorParameter() {parameterName=PanSpeed1, defaultValue=<rgba -0.2 0.6 0 0>}.RGB) {g=true, r=true}
  ) {speedX=-1, speedY=2.5}
  n9 = ScalarParameter() {defaultValue=0.01, parameterName=ChromaOffset}
  n10 = TextureSampleParameter2D(UVs=Add(A=@n8, B=Multiply(A=@n6, B=@n9))) {parameterName=NoiseB, texture=/Game/Effects/Fort_Effects/Textures/Tile/T_Inky_Smoke_Tile}
  n11 = ComponentMask(None=VectorParameter() {parameterName=Tile2, defaultValue=<rgba 4 2 0 0>}.RGB) {g=true, r=true}
  n12 = Panner(
   Coordinate=Multiply(A=TextureCoordinate(), B=@n11)
   Time=@n7
   Speed=ComponentMask(None=VectorParameter() {parameterName=PanSpeed2, defaultValue=<rgba 0.15 0.3 0 0>}.RGB) {g=true, r=true}
  ) {speedX=1, speedY=1}
  n13 = TextureSampleParameter2D(UVs=Add(A=Multiply(A=@n9, B=@n11), B=@n12)) {parameterName=NoiseA, texture=&T_Inky_Smoke_Tile}
  n14 = TextureSampleParameter2D(UVs=@n8) {parameterName=NoiseB, texture=&T_Inky_Smoke_Tile}
  n15 = TextureSampleParameter2D(UVs=@n12) {parameterName=NoiseA, texture=&T_Inky_Smoke_Tile}
  n16 = LinearInterpolate(
   A=Multiply(A=Constant3Vector() {constant=<rgba 1 0 0.4 0>}, B=Add(A=@n10.G, B=@n13.G))
   B=Multiply(A=Constant3Vector() {constant=<rgba 0.0245 1 0.5102 0>}, B=Add(A=@n14.G, B=@n15.G))
  )
  n17 = ScalarParameter() {defaultValue=2.5, parameterName=Power}
  Multiply(A=Desaturation(None=Power(Base=@n16, Exp=@n17), Fraction=ScalarParameter() {defaultValue=-0.25, parameterName=Desaturation}), B=@n1.RGB)
 ]
}
