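# Unlit translucent particle material for the NightNight effects, drawing a rocket with an engine glow and trail moving along a strip.
# A dynamic parameter's Travel value slides the rocket texture and two sphere masks along V, with separate size inputs for glow and trail.
# The masks blend two colors tinted by particle color, with HDR derived from LDR and a feature-level fallback.
# SplineThicken sets strip width, gated by travel position.
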Material /Figment_S10/Athena/Prototype/Blueprints/NightNight/FX/M_NightNight_RocketSpline
graph {
 properties {
  bUsedWithMeshParticles=true
  bUsedWithParticleSprites=true
  blendMode=BLEND_Translucent
  pixelDepthOffsetMode=PDOM_Legacy
  shadingModel=MSM_Unlit
  twoSided=true
  usedShadingModels=Unlit
 }
 nodes [
  n1 = ScalarParameter() {defaultValue=50, parameterName=RocketSize}
  n2 = Multiply(A=TextureCoordinate(), B=AppendVector(A=Constant() {r=1}, B=@n1))
  n3 = Constant2Vector() {g=0.95}
  n4 = Constant()
  n5 = DynamicParameter() {defaultValue=<rgba 0.5 1 1 1>, paramNames [Travel,TrailUVSize,GlowUVSize,Param4]}
  n6 = Add(A=@n5.Travel) {constB=0.25}
  n7 = Multiply(A=LinearInterpolate(A=@n1, Alpha=@n6) {constB=0}, B=ScalarParameter() {defaultValue=1.5, parameterName=TravelMult})
  n8 = TextureSample(UVs=Subtract(A=Add(A=@n2, B=@n3), B=AppendVector(A=@n4, B=@n7))) {texture=&/T_NightNight_Rocket_D}
  n9 = Constant() {r=1}
  n10 = Multiply(A=ScalarParameter() {defaultValue=60, parameterName=Y Size Engine}, B=@n5.GlowUVSize)
  n11 = LinearInterpolate(A=@n10, Alpha=@n6) {constA=-0.5, constB=0}
  n12 = ScalarParameter() {defaultValue=1.5, parameterName=TravelMult}
  n13 = SphereMask(
   A=Multiply(A=TextureCoordinate(), B=AppendVector(A=@n9, B=@n10))
   B=AppendVector(A=Constant() {r=0.5}, B=Multiply(A=@n11, B=@n12))
   Radius=Constant() {r=0.5}
  ) {hardnessPercent=0}
  n14 = TextureCoordinate()
  n15 = Multiply(A=@n5.TrailUVSize, B=ScalarParameter() {defaultValue=1.5, parameterName=Y Size trail})
  n16 = AppendVector(A=Constant() {r=1}, B=@n15)
  n17 = Constant() {r=0.5}
  n18 = Multiply(A=@n12, B=LinearInterpolate(A=@n15, Alpha=@n6) {constA=1, constB=0})
  n19 = Subtract(A=ComponentMask(None=TextureCoordinate()) {g=true}, B=Multiply(A=@n12, B=LinearInterpolate(Alpha=@n6) {constA=1, constB=0}))
  n20 = Multiply(
   A=SphereMask(A=Multiply(A=@n14, B=@n16), B=AppendVector(A=@n17, B=@n18), Radius=ScalarParameter() {defaultValue=0.25, parameterName=TrailWidth}) {hardnessPercent=0}
   B=Ceil(None=@n19)
  )
  n21 = Saturate(None=Power(Base=Add(A=@n13, B=@n20)) {constExponent=1.5})
  Add(A=@n8.A, B=@n21)
  Divide(A=Time() {bOverride_Period=true, period=2})
  n22 = LinearInterpolate(
   A=VectorParameter() {parameterName=Color2, defaultValue=<rgba 0.2982 0.08 1 0>}.RGB
   B=VectorParameter() {parameterName=Color1, defaultValue=<rgba 0.0796 1.0875 2 0>}.RGB
   Alpha=@n21
  )
  n23 = LinearInterpolate(A=Multiply(A=ParticleColor().RGB, B=Multiply(A=@n22, B=@n21)), B=@n8.RGB, Alpha=@n8.RGB)
  n24 = Multiply(A=@n23, B=Add(A=Multiply(A=@n13) {constB=1.5}))
  n25 = MaterialFunctionCall(LDR Input=Clamp(None=@n23), HDR Tint=@n24) {materialFunction=&DeriveHDRfromLDR}
  FeatureLevelSwitch(Default=@n25.RemappedRange, ES2=@n24, ES3_1=@n24, SM4_REMOVED=@n25.RemappedRange)
  n26 = Multiply(A=Ceil(None=Add(A=@n19) {constB=0.1}), B=ScalarParameter() {defaultValue=1800, parameterName=Width})
  MaterialFunctionCall(WidthBase=@n26, WidthTip=@n26) {materialFunction=&SplineThicken}
 ]
}
