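# Masked, Default Lit mesh-particle material for the NightNight smoke wave effect.
# Two panning, scaled noise textures, shifted vertically by a particle-driven Speed parameter, multiply into one smoke pattern.
# That pattern blends Dark and HightLight colors and, weighted by a sine of the V coordinate that fades both edges, drives a dissolve value offset by the particle's Dissovle parameter.
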Material /Figment_S10/Athena/Prototype/Blueprints/NightNight/FX/M_NightNight_SmokeWave
graph {
 properties {
  bUsedWithMeshParticles=true
  blendMode=BLEND_Masked
  pixelDepthOffsetMode=PDOM_Legacy
  previewMesh=&/SM_NightNight_SmokeWave
  usedShadingModels=Default Lit
 }
 nodes [
  Constant()
  n1 = DynamicParameter() {paramNames [Speed,Dissovle,Param3,Param4]}
  n2 = VectorParameter() {defaultValue=<rgba 3 0.2 0 -1>}
  n3 = Multiply(A=Panner(Speed=AppendVector(A=@n2.B, B=@n2.B)), B=AppendVector(A=@n2.R, B=@n2.R))
  n4 = AppendVector(A=Constant(), B=Multiply(A=DynamicParameter() {paramNames [Speed,Dissovle,Param3,Param4]}.Speed) {constB=-1})
  n5 = VectorParameter() {parameterName=Param2, defaultValue=<rgba 3 0.5 0.05 -0.5>}
  n6 = Multiply(A=Panner(Speed=AppendVector(A=@n5.B, B=@n5.B)), B=AppendVector(A=@n5.R, B=@n5.R))
  n7 = Multiply(
   A=TextureSample(UVs=Add(A=@n3, B=@n4)) {texture=/Figment_S10/Environments/Hexmap/HexTile/T_FireNoise_01}.R
   B=TextureSample(UVs=Add(A=@n4, B=@n6)) {samplerType=SAMPLERTYPE_LinearColor, texture=/Game/Weapons/WeaponSkins/Textures/T_Swirly_Noise_highContrast1}.R
  )
  n8 = Multiply(A=ComponentMask(None=TextureCoordinate()) {g=true}) {constB=0.5}
  LinearInterpolate(A=@n1.Dissovle, B=Add(A=@n1.Dissovle), Alpha=Multiply(A=@n7, B=Sine(None=@n8)))
  LinearInterpolate(
   A=VectorParameter() {parameterName=Dark, defaultValue=<rgba 0.3438 0.1331 0.1957 1>}.RGB
   B=VectorParameter() {parameterName=HightLight, defaultValue=<rgba 1 0.821 0.4532 1>}.RGB
   Alpha=@n7
  )
  ParticleColor()
 ]
}
