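# Translucent, two-sided particle sprite material for the NightNight prototype, apparently an outward-moving smoke wave.
# Its noise UVs pair sprite U with a world-space sphere mask around the actor position, so the pattern follows distance from it.
# Two panning textures, offset by a per-particle random dynamic parameter, multiply together with a sine falloff across U.
# Particle alpha scales the result, with depth fade skipped on ES2, ES3_1 and SM4.
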Material /Figment_S10/Athena/Prototype/Blueprints/NightNight/FX/M_NightNight_SmokeWaveOut
graph {
 properties {bUsedWithParticleSprites=true, blendMode=BLEND_Translucent, pixelDepthOffsetMode=PDOM_Legacy, twoSided=true, usedShadingModels=Default Lit}
 nodes [
  Constant3Vector() {constant=<rgba 1 1 1 0>}
  n1 = AppendVector(
   A=ComponentMask(None=TextureCoordinate()) {r=true}
   B=SphereMask(
    A=WorldPosition().XYZ
    B=ActorPositionWS()
    Radius=ScalarParameter() {defaultValue=10000, parameterName=Radius}
    Hardness=ScalarParameter() {parameterName=Hardness}
   )
  )
  n2 = VectorParameter() {parameterName=Texture1_Ctrl, defaultValue=<rgba 0.5 0.5 0 1>}
  n3 = AppendVector(A=@n2.B, B=@n2.B)
  n4 = DynamicParameter() {paramNames [UV_random,Param2,Param3,Param4]}
  n5 = Add(A=Multiply(A=Panner(Coordinate=@n1, Speed=@n3), B=AppendVector(A=@n2.R, B=@n2.R)), B=@n4.UV_random)
  n6 = VectorParameter() {parameterName=Texture2_Ctrl, defaultValue=<rgba 0.3 0.5 0 1.2>}
  n7 = AppendVector(A=@n6.B, B=@n6.B)
  n8 = Add(A=@n4.UV_random, B=Multiply(A=Panner(Coordinate=@n1, Speed=@n7), B=AppendVector(A=@n6.R, B=@n6.R)))
  n9 = TextureCoordinate()
  n10 = MaterialFunctionCall(Float2=Sine(None=Multiply(A=@n9) {constB=0.5})) {materialFunction=&BreakOutFloat2Components}
  n11 = Multiply(
   A=Multiply(
    A=TextureSampleParameter2D(UVs=@n5) {parameterName=Texture1, texture=/Game/VisualThreatManager/StormVisuals/Textures/2ndInfectedMapLinear}.RGB
    B=TextureSampleParameter2D(UVs=@n8) {parameterName=Texture2, texture=/Figment_S10/Effects/Fort_Effects/Textures/AtmosphericCloudNoise01}.R
   )
   B=Power(Base=Multiply(A=@n10.R, B=@n10.R))
  )
  n12 = DepthFade(Opacity=@n11) {fadeDistanceDefault=500}
  Multiply(A=ParticleColor().A, B=FeatureLevelSwitch(Default=@n12, ES2=@n11, ES3_1=@n11, SM4_REMOVED=@n12))
 ]
}
