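# Translucent unlit material for the NightNight Zero Point explosion sphere, a glitching plasma bubble.
# World position offset combines masked glitch displacement along normals with a height-weighted squash driven by BubbleSmush.
# Emissive pans two offset turbulence samples for chromatic fringing, using screen-radial and Fresnel UVs with camera-spin compensation, plus glitch and edge glow.
# Opacity combines Fresnel, glitch, smush and depth fade; low quality simplifies colour and skips spin compensation.
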Material /Figment_S10/Athena/Prototype/Blueprints/NightNight/FX/M_NightNight_ZeroPoint_Explo_Sphere
graph {
 properties {
  bUsedWithInstancedStaticMeshes=true
  bUsedWithMeshParticles=true
  blendMode=BLEND_Translucent
  pixelDepthOffsetMode=PDOM_Legacy
  shadingModel=MSM_Unlit
  usedShadingModels=Unlit
 }
 nodes [
  n1 = VertexNormalWS()
  n2 = ScalarParameter() {parameterName=BubbleSmush}
  n3 = Multiply(A=Add(A=@n2), B=ScalarParameter() {defaultValue=1000, parameterName=GlitchWPO})
  n4 = Multiply(
   A=MaterialFunctionCall!(MaskScaleMult=ScalarParameter() {defaultValue=2, parameterName=GlitchScale}) {materialFunction=/Game/Athena/Effects/Materials/Meteor/MF_Glitch_FX_World_Mask}.WorldSpaceMask
   B=ScalarParameter() {parameterName=Glitch}
  )
  n5 = Subtract(A=WorldPosition().XYZ, B=ObjectPositionWS())
  n6 = Saturate(None=Divide(A=ComponentMask(None=@n5) {b=true}, B=ObjectRadius()))
  n7 = Multiply(A=ComponentMask(None=VertexNormalWS()) {g=true, r=true}, B=ScalarParameter() {defaultValue=-4000, parameterName=HorizontalSmush})
  n8 = Multiply(A=Power(Base=@n6) {constExponent=1}, B=AppendVector(A=Multiply(A=@n7) {constB=-2}, B=Constant()))
  n9 = AppendVector(A=@n7, B=ScalarParameter() {defaultValue=-14000, parameterName=VertialSmush})
  n10 = LinearInterpolate(
   A=ScalarParameter() {defaultValue=30, parameterName=SmushStartPower}
   B=ScalarParameter() {defaultValue=10, parameterName=SmushEndPower}
   Alpha=@n2
  ) {constA=2}
  n11 = Power(Base=@n6, Exp=Multiply(A=@n10) {constB=1.25})
  Add(A=Multiply(A=Multiply(A=@n1, B=@n3), B=@n4), B=Multiply(A=@n2, B=LinearInterpolate(A=@n8, B=@n9, Alpha=@n11)))
  n12 = ScalarParameter() {defaultValue=0.25, parameterName=CenterOpacityAdd}
  n13 = Multiply(A=ScalarParameter() {defaultValue=0.3, parameterName=GlitchOpacityAdd}, B=@n4) {constB=0.25}
  n14 = Fresnel(ExponentIn=ScalarParameter() {defaultValue=0.5, parameterName=UVFresnelExp}) {baseReflectFraction=0, exponent=1}
  n15 = Saturate(None=LinearInterpolate(Alpha=@n14) {constA=1.6, constB=-0.15})
  n16 = Multiply(A=Add(A=@n14, B=Add(A=@n12, B=@n13)) {constB=0.25}, B=@n15)
  n17 = Multiply(A=@n2, B=Power(Base=@n6, Exp=@n10) {constExponent=30})
  n18 = Multiply(A=@n17, B=ScalarParameter() {defaultValue=3, parameterName=SmushOpacityAdd}) {constB=0.1}
  n19 = DepthFade(FadeDistance=ScalarParameter() {defaultValue=1600, parameterName=FadeDistance})
  Multiply(A=Multiply(A=Add(A=@n16, B=@n18), B=@n19), B=ScalarParameter() {defaultValue=1, parameterName=Opacity})
  n20 = Multiply(A=@n4, B=@n4)
  n21 = VectorParameter() {parameterName=GlitchColor, defaultValue=<rgba 0.0365 1 0.9552 0>}
  n22 = Add(A=@n14, B=OneMinus(None=@n19))
  n23 = Multiply(
   A=Add(A=Multiply(A=@n21.RGB, B=@n20), B=VectorParameter() {parameterName=Background, defaultValue=<rgba 0.0196 0.0365 0.25 0>}.RGB)
   B=Multiply(A=@n22, B=@n22)
  )
  n24 = ScalarParameter() {defaultValue=-0.2, parameterName=ChromaOffset}
  n25 = LinearInterpolate(Alpha=@n20) {constA=0.25, constB=3.5}
  n26 = AppendVector(A=Constant() {r=1}, B=Add(A=Multiply(A=@n24, B=@n25)))
  n27 = VectorParameter() {parameterName=Tile, defaultValue=<rgba 2 0.6 0 0>}
  n28 = Subtract(
   A=ScreenPosition().ViewportUV
   B=MaterialFunctionCall(World Position=ActorPositionWS()) {materialFunction=&TransformToClipSpace}.Buffer UVs for Scene Textures
  )
  n29 = ComponentMask(None=Saturate(None=MaterialFunctionCall(Vector or UVs=@n28) {materialFunction=&VectorToRadialValue}.Radial Coordinates)) {r=true}
  n30 = ScalarParameter() {defaultValue=1, parameterName=SpinCounteract}
  n31 = Constant()
  n32 = Constant3Vector() {constant=<rgba 1 0 0 0>}
  n33 = MaterialFunctionCall(Vector or UVs=ComponentMask(None=Transform(None=@n32) {transformSourceType=TRANSFORMSOURCE_View}) {g=true, r=true}) {materialFunction=&VectorToRadialValue}
  n34 = Constant()
  n35 = QualitySwitch(Default=Multiply(A=AppendVector(A=@n30, B=@n31), B=ComponentMask(None=@n33.Radial Coordinates) {r=true}) {constB=-1}, Low=@n34)
  n36 = Add(
   A=Multiply(A=AppendVector(A=@n27.R, B=@n27.R), B=AppendVector(A=@n29, B=@n14))
   B=FeatureLevelSwitch(Default=@n35, ES2=@n34, ES3_1=@n34, SM4_REMOVED=@n35)
  )
  n37 = Add(A=ScalarParameter() {parameterName=PanAdd}, B=Multiply(A=Time(), B=ScalarParameter() {defaultValue=0.5, parameterName=PanSpeed}))
  n38 = TextureCoordinate()
  n39 = DDX(Value=@n38)
  n40 = DDY(Value=@n38)
  n41 = TextureSample!(
   UVs=Panner(Coordinate=Multiply(A=@n26, B=@n36), Time=@n37) {speedX=0.015, speedY=-0.2}
   Tex
   "DDX(UVs)"=@n39
   "DDY(UVs)"=@n40
   Apply View MipBias
  ) {mipValueMode=TMVM_Derivative, texture=/Game/Effects/Fort_Effects/Textures/Plasma/T_Turbulence_Stretched}
  n42 = VectorParameter() {parameterName=Chroma1, defaultValue=<rgba 0.0234 1.3266 1.5 1>}
  n43 = TextureSample(UVs=Panner(Coordinate=@n36, Time=@n37) {speedX=0.015, speedY=-0.2}, "DDX(UVs)"=@n39, "DDY(UVs)"=@n40) {mipValueMode=TMVM_Derivative, texture=&T_Turbulence_Stretched}
  n44 = VectorParameter() {parameterName=Chroma2, defaultValue=<rgba 1 0 0.2339 1>}
  n45 = Power(
   Base=LinearInterpolate(A=Multiply(A=@n41.G, B=@n42.RGB), B=Multiply(A=@n43.G, B=@n44.RGB))
   Exp=LinearInterpolate(
    A=ScalarParameter() {defaultValue=1.4, parameterName=PowerInner}
    B=ScalarParameter() {defaultValue=0.2, parameterName=PowerOuter}
    Alpha=@n14
   ) {constA=1.5, constB=0.15}
  )
  n46 = Multiply(A=VectorParameter() {parameterName=LowColor, defaultValue=<rgba 0.9318 0.5876 1.1 0>}.RGB, B=@n41.G)
  n47 = LinearInterpolate(A=@n42.RGB, B=@n44.RGB, Alpha=@n15)
  n48 = Add(A=Fresnel(ExponentIn=ScalarParameter() {defaultValue=5, parameterName=ExponentIn}) {baseReflectFraction=0}, B=Multiply(A=@n17) {constB=2})
  n49 = Add(
   A=Multiply(A=QualitySwitch(Default=@n45, Low=@n46), B=LinearInterpolate(Alpha=@n14) {constA=2, constB=0}) {constB=3}
   B=Multiply(A=Multiply(A=@n47, B=@n48), B=ScalarParameter() {defaultValue=3, parameterName=EdgeMult})
  )
  n50 = ScalarParameter() {defaultValue=1, parameterName=OverallEmissiveMuti}
  Multiply(
   A=LinearInterpolate(B=ScalarParameter() {defaultValue=10, parameterName=GlitchBrightness}, Alpha=@n20) {constA=0.8, constB=1.5}
   B=Multiply(A=Add(A=@n23, B=@n49), B=Add(A=@n50, B=@n2))
  )
 ]
}
