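# Masked unlit material for a rocket smoke trail pillar, with height, top size and fade driven by MPC_Snow.
# Two panning storm textures, mapped by world Z and stretched near the top, supply noise and normals to MF_Giant_Smoke_Pillar_Color.
# A height gradient scales noisy horizontal vertex displacement and tapers the top toward the actor's axis.
# Opacity erodes top-first with fade, drops near the camera and uses temporal dither outside low-end paths.
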Material /Figment_S10/Athena/Prototype/Blueprints/NightNight/FX/M_Rocket_Trail
graph {
 properties {
  bUsedWithInstancedStaticMeshes=true
  blendMode=BLEND_Masked
  pixelDepthOffsetMode=PDOM_Legacy
  shadingModel=MSM_Unlit
  usedShadingModels=Unlit
 }
 nodes [
  n1 = TextureCoordinate() {vTiling=2}
  n2 = Divide(A=ComponentMask(None=WorldPosition().XYZ) {b=true}, B=ScalarParameter() {defaultValue=20000, parameterName=VerticalTile})
  n3 = Add(
   A=CollectionParameter() {collection=/Game/Athena/Prototype/Blueprints/White/MPC_Snow, parameterName=RocketSmokeHeight}
   B=ScalarParameter() {defaultValue=1500, parameterName=HeightAdd}
  ) {constB=500}
  n4 = Subtract(A=@n3, B=ComponentMask(None=WorldPosition().XYZ) {b=true})
  n5 = ScalarParameter() {defaultValue=9000, parameterName=WPOGradientSize}
  n6 = CollectionParameter() {collection=&MPC_Snow, parameterName=RocketSmokeTopSize}
  n7 = Saturate(None=Divide(A=@n4, B=Multiply(A=@n5, B=@n6)))
  n8 = LinearInterpolate(
   A=Multiply(
    A=ScalarParameter() {defaultValue=-0.05, parameterName=TopUVStretch}
    B=CollectionParameter() {collection=&MPC_Snow, parameterName=RocketSmokeTopSize}
   )
   Alpha=Power(Base=@n7) {constExponent=5}
  ) {constB=0}
  n9 = ComponentMask(None=VectorParameter() {parameterName=Pan Speed, defaultValue=<rgba 0.02 0.05 0 1>}.RGB) {g=true, r=true}
  n10 = Panner(
   Coordinate=AppendVector(
    A=MaterialFunctionCall(Float2=@n1) {materialFunction=&BreakOutFloat2Components}.R
    B=Add(A=@n2, B=@n8)
   )
   Speed=@n9
  ) {bFractionalPart=true, speedX=0.025, speedY=-1}
  n11 = TextureSample(UVs=@n10) {texture=/Game/VisualThreatManager/StormVisuals/Textures/2ndInfectedMapLinear}
  n12 = TextureCoordinate() {uTiling=2, vTiling=4}
  n13 = Multiply(A=@n2) {constB=1.7}
  n14 = Panner(
   Coordinate=AppendVector(
    A=MaterialFunctionCall(Float2=@n12) {materialFunction=&BreakOutFloat2Components}.R
    B=Add(A=@n13, B=@n8)
   )
   Speed=@n9
  ) {bFractionalPart=true, speedX=-0.05, speedY=-1}
  n15 = TextureSample(UVs=@n14) {texture=&2ndInfectedMapLinear}
  n16 = LinearInterpolate(Alpha=Add(A=@n11.G, B=@n15.G)) {constA=-1, constB=1.5}
  n17 = Normalize(VectorInput=Multiply(A=VertexNormalWS(), B=Constant3Vector() {constant=<rgba 1 1 0 0>}))
  n18 = ScalarParameter() {defaultValue=150, parameterName=Deform}
  n19 = Subtract(A=ActorPositionWS(), B=WorldPosition().XYZ)
  n20 = Constant3Vector() {constant=<rgba 1 1 0 0>}
  Add(
   A=Multiply(A=Multiply(A=@n16, B=@n17), B=Multiply(A=@n18, B=@n7))
   B=LinearInterpolate(A=Multiply(A=@n19, B=@n20), Alpha=Power(Base=@n7) {constExponent=0.6}) {constB=0}
  )
  n21 = ScalarParameter() {defaultValue=0.85, parameterName=Dither Threshold}
  n22 = CameraPositionWS()
  n23 = WorldPosition()
  n24 = LinearInterpolate(A=@n11.G, B=@n15.G)
  n25 = ScalarParameter() {defaultValue=5, parameterName=SmokeFadePower}
  n26 = Saturate(None=CollectionParameter() {collection=&MPC_Snow, parameterName=RocketSmokeFade})
  n27 = LinearInterpolate(Alpha=@n7) {constA=-2, desc=Erode top first}
  n28 = Multiply(
   A=Power(Base=@n24, Exp=LinearInterpolate(B=@n25, Alpha=@n26) {constA=0.8})
   B=Saturate(None=LinearInterpolate(B=@n27, Alpha=@n26) {constA=1})
  )
  n29 = LinearInterpolate(Alpha=@n26) {constA=1, constB=0}
  n30 = Multiply(
   A=Divide(A=Distance(A=@n22, B=@n23.XYZ), B=ScalarParameter() {defaultValue=800, parameterName=Camera Distance Fade Falloff})
   B=Multiply(A=Multiply(A=@n28, B=@n29), B=Saturate(None=@n4) {desc=Mask off top})
  )
  n31 = QualitySwitch(
   Default=Multiply(
    A=MaterialFunctionCall(Alpha Threshold=@n21) {materialFunction=&DitherTemporalAA}.Result
    B=@n30
   )
   Low=@n30
  )
  FeatureLevelSwitch(Default=@n31, ES2=@n30, ES3_1=@n30, SM4_REMOVED=@n31)
  n32 = MaterialFunctionCall(
   BaseNormal=TextureSample(UVs=@n10) {samplerType=SAMPLERTYPE_Normal, texture=/Game/VisualThreatManager/StormVisuals/Textures/2ndInfectedMapNormal}.RGB
   AdditionalNormal=TextureSample(UVs=@n14) {samplerType=SAMPLERTYPE_Normal, texture=&2ndInfectedMapNormal}.RGB
  ) {materialFunction=&BlendAngleCorrectedNormals}
  n33 = LinearInterpolate(Alpha=@n26) {constA=1, constB=-2}
  Reroute(None=MaterialFunctionCall!(
    GlowDirection
    GlowStartHeight=@n3
    GlowWorldPosition
    Normal=Transform(None=@n32.Result)
    Mask=@n24
    Base Glow Mult=Saturate(None=@n33)
    Inner Glow Mult
    GlowDirectionBlend
   ) {materialFunction=/Game/Athena/Prototype/Blueprints/White/FX/MF_Giant_Smoke_Pillar_Color}.Result)
 ]
}
