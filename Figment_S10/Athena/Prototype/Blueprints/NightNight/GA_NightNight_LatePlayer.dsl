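; Gameplay ability that catches a late player up to the NightNight live event.
; It applies knock and block effects, sets flying and ghost mode, plays a montage, hides the storm client-side and stops building-foundation streaming.
; It reads the replicated level sequence's time to schedule what remains: a launch up and away from center before the second bump, or a timer before the pull.
; Both lead to movement effects, input lock outside replays and a camera modifier.
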(event Ability|EventActivateAbility
  (bind _returnvalue (Ability|GetActivatingPawn))
  (bind _returnvalue_2 (Ability|GetActivatingPawn))
  (bind _returnvalue_3 (Ability|GetAbilitySystemComponent _returnvalue_2))
  (bind _returnvalue_4 (Ability|GetActivatingPawn))
  (bind _returnvalue_5 (Ability|GetAbilitySystemComponent _returnvalue_4))
  (bind _secondssinceeventbegan (Variables|Default|GetSecondsSinceEventBegan))
  (bind _pull (Variables|Default|GetPull))
  (Utilities|FlowControl|Sequence
    (:then_0
      (GameplayEffects|ApplyGameplayEffectToTarget :self _returnvalue_3 :GameplayEffectClass "/Figment_S10/Athena/Prototype/Blueprints/NightNight/GE_NightNight_KnockPlayer_GC.GE_NightNight_KnockPlayer_GC_C" :Target _returnvalue_3 :Level 0.0)
      (GameplayEffects|ApplyGameplayEffectToTarget :self _returnvalue_3 :GameplayEffectClass "/Figment_S10/Athena/Prototype/Blueprints/NightNight/GE_NightNight_Sblock.GE_NightNight_Sblock_C" :Target _returnvalue_3 :Level 0.0))
    (:then_1
      (Utilities|FlowControl|Delay 1.0)
      (Weapon|SetHolsterWeapon _returnvalue 0 true false "None")
      (Pawn|Components|CharacterMovement|SetMovementMode (Variables|/Script/Engine.Character|GetCharacterMovement (Ability|GetActivatingPawn)) "MOVE_Flying" 0)
      (bind _asfort_player_controller_gameplay (Utilities|Casting|CastToFortPlayerControllerGameplay (Fortnite|GetFortPlayerControllerfromActor _returnvalue)))
      (Controller|GhostMode|StartGhostMode _asfort_player_controller_gameplay "/Game/Athena/Items/Gameplay/NightNight/AGID_NightNight.AGID_NightNight" :!class /Script/FortniteGame.FortPlayerControllerGameplay)
      (Animation|PlayAnimMontage (Ability|GetActivatingPawn) "/Figment_S10/Animation/Game/MainPlayer/Combat/Gadgets/Medium/Male/GhostRock/GhostRock_Idle_CMM_Montage.GhostRock_Idle_CMM_Montage" 1.0 "None"))
    (:then_2
      ;[1 1 1] Hide the storm
      (if (not (Networking|IsServer))
        (bind _outactors (Actor|GetAllActorsOfClass "/Game/Athena/SafeZone/BP_SafeZoneStorm.BP_SafeZoneStorm_C"))
        (Variables|Default|SetSafezoneStormBP (Utilities|Array|Get(acopy) _outactors))
        (Class|BPSafeZoneStorm|ShowStormScreenEffect (Variables|Default|GetSafezoneStormBP) false)))
    (:then_3
      (bind _asfort_player_controller_athena (Utilities|Casting|CastToFortPlayerControllerAthena (Fortnite|GetFirstLocalFortPlayerController true true)))
      (Streaming|SetCanStreamBuildingFoundationsIn _asfort_player_controller_athena false))
    (:then_4
      (bind _outactors_2 (Actor|GetAllActorsOfClass "/Game/Athena/Prototype/Blueprints/Slab/Replicated_LevelSequence.Replicated_LevelSequence_C"))
      (Variables|Default|SetSecondsSinceEventBegan (Utilities|TimeManagement|QualifiedFrameTimeToSeconds (Sequencer|Player|GetCurrentTime (Variables|/Script/LevelSequence.LevelSequenceActor|GetSequencePlayer (Utilities|Array|Get(acopy) _outactors_2)))))
      (Variables|Default|SetBeforeSecondBump (< (Variables|Default|GetSecondsSinceEventBegan) (Variables|Default|GetSecondBump)))
      (if (Variables|Default|GetBeforeSecondBump)
        (Variables|Default|SetTimeBeforeSecondBump (- (Variables|Default|GetSecondBump) (Variables|Default|GetSecondsSinceEventBegan)))
        (Utilities|FlowControl|Delay (Variables|Default|GetTimeBeforeSecondBump))
        (Ability|Tasks|MoveToLocation "None" (Math|Vector|vector+vector (Transformation|GetActorLocation (Ability|GetActivatingPawn)) (* (Math|Vector|GetForwardVector (Math|Rotator|FindLookatRotation (Variables|Default|GetCenterLoc) (Transformation|GetActorLocation (Ability|GetActivatingPawn)))) (Variables|Default|GetMove2Distance)) "0,0,15000.000000") (Variables|Default|GetMoveInDuration) "/Figment_S10/Athena/Prototype/Blueprints/NightNight/NightNight_MovementInCurve.NightNight_MovementInCurve"
          (:then
            (bind _returnvalue_6 (GameplayEffects|ApplyGameplayEffectToTarget :self (Ability|GetAbilitySystemComponent (Ability|GetActivatingPawn)) :GameplayEffectClass "/Figment_S10/Athena/Prototype/Blueprints/NightNight/GE_NightNight_SlowMovePlayer_GC.GE_NightNight_SlowMovePlayer_GC_C" :Target (Ability|GetAbilitySystemComponent (Ability|GetActivatingPawn)) :Level 0.0))
            (bind _output_get (Variables|Default|SetSlowMovementGEforGC _returnvalue_6))
            (Utilities|FlowControl|Delay 3.0)
            (GameplayEffects|RemoveActiveGameplayEffect _returnvalue_5 _output_get -1)
            (Utilities|FlowControl|Delay 13.0)
            (bind _returnvalue_7 (GameplayEffects|ApplyGameplayEffectToTarget :self _returnvalue_5 :GameplayEffectClass "/Figment_S10/Athena/Prototype/Blueprints/NightNight/GE_NightNight_MovePlayer_GC.GE_NightNight_MovePlayer_GC_C" :Target _returnvalue_5 :Level 0.0))
            (Variables|Default|SetMovementGEforGC _returnvalue_7)
            (Variables|Default|SetLocalPlayerController (Fortnite|GetFortPlayerControllerfromActor (Ability|GetActivatingPawn)))
            (Utilities|FlowControl|Sequence
              (:then_0
                (if (not (Replay|GetIsPlayingReplay))
                  (Input|SetIgnoreLookInput (Variables|Default|GetLocalPlayerController) true)
                  (Input|SetIgnoreMoveInput (Variables|Default|GetLocalPlayerController) true)))
              (:then_1
                (Game|Player|AddNewCameraModifier (Variables|/Script/Engine.PlayerController|GetPlayerCameraManager (Variables|Default|GetLocalPlayerController)) "/Figment_S10/Athena/Prototype/Blueprints/NightNight/NIghtNight_Camera_Modifier.NIghtNight_Camera_Modifier_C"))))
          (:OnTargetLocationReached
            (Pawn|Components|CharacterMovement|SetMovementMode (Variables|/Script/Engine.Character|GetCharacterMovement (Ability|GetActivatingPawn)) "MOVE_Flying" 0)))
        (elif (< _secondssinceeventbegan _pull)
          (Utilities|Time|SetTimerbyEvent (EventDispatchers|CreateEvent :!event Custom|LateCameraTricks) (- _pull _secondssinceeventbegan)))))))

(event Custom|LateCameraTricks
  (bind _returnvalue (GameplayEffects|ApplyGameplayEffectToTarget :self (Ability|GetAbilitySystemComponent (Ability|GetActivatingPawn)) :GameplayEffectClass "/Figment_S10/Athena/Prototype/Blueprints/NightNight/GE_NightNight_MovePlayer_GC.GE_NightNight_MovePlayer_GC_C" :Target (Ability|GetAbilitySystemComponent (Ability|GetActivatingPawn)) :Level 0.0))
  (Variables|Default|SetMovementGEforGC _returnvalue)
  (Variables|Default|SetLocalPlayerController (Fortnite|GetFortPlayerControllerfromActor (Ability|GetActivatingPawn)))
  (Utilities|FlowControl|Sequence
    (:then_0
      (if (not (Replay|GetIsPlayingReplay))
        (Input|SetIgnoreLookInput (Variables|Default|GetLocalPlayerController) true)
        (Input|SetIgnoreMoveInput (Variables|Default|GetLocalPlayerController) true)))
    (:then_1
      (Game|Player|AddNewCameraModifier (Variables|/Script/Engine.PlayerController|GetPlayerCameraManager (Variables|Default|GetLocalPlayerController)) "/Figment_S10/Athena/Prototype/Blueprints/NightNight/NIghtNight_Camera_Modifier.NIghtNight_Camera_Modifier_C"))))
