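; NightNight prototype camera modifier that gradually turns the view toward a fixed world point while pulling it back.
; BlueprintModifyCamera accumulates delta time to drive two clamped lerps, leaving FOV unchanged.
; Location eases 1500 units away from that point along the look-at axis over 30 seconds.
; Rotation takes the shortest path to the look-at rotation, timed separately by TimeToDesiredRotation.
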(fn BlueprintModifyCamera
  (params (DeltaTime :type Float%20%28single-precision%29) (ViewLocation :type Vector :default "0, 0, 0") (ViewRotation :type Rotator :default "0, 0, 0") (FOV :type Float%20%28single-precision%29))
  (Variables|Default|SetTimeAccumulation (+ DeltaTime (Variables|Default|GetTimeAccumulation)))
  (return (Math|Vector|Lerp(Vector) ViewLocation (+ ViewLocation (* (Math|Vector|GetForwardVector (Math|Rotator|FindLookatRotation ViewLocation (66048.0 -43008.0 0.0))) -1.5e+03)) (Math|Float|Clamp(Float) (/ (Variables|Default|GetTimeAccumulation) 3e+01))) (Math|Rotator|Lerp(Rotator) ViewRotation (Math|Rotator|FindLookatRotation ViewLocation (66048.0 -43008.0 0.0)) (Math|Float|Clamp(Float) (/ (Variables|Default|GetTimeAccumulation) (Variables|Default|GetTimeToDesiredRotation))) true) FOV))
