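# Level Sequence for part 2 of the NightNight prototype, animating seven separated rockets (BP_Geode_NightNight_Separated) in staggered, overlapping flights.
# Each rocket spawns alongside its own CameraRig_Rail, whose keyed rail position serves as the flight path, with an attach track on the rocket.
# Each MissileSilo_Stage2_Main mesh shrinks to zero as its flight ends, shortly before despawn.
# Rocket camera-shake components hold a constant strength, and a root event track carries a trigger.
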LevelSequence /Figment_S10/Athena/Prototype/Blueprints/NightNight/NightNight_Part2_Animation
bindings [
 NightNight_SmallerRocket_CameraShakeComponent {
  id=#1
  parent_id=#2
  possessed_class=NightNight_SmallerRocket_CameraShakeComponent_C
  tracks [{
    type=MovieSceneFloatTrack
    display_name=Shake Strength
    sections [{type=MovieSceneFloatSection, channels [None {type=MovieSceneScriptingFloatChannel, default=0.5, const [0.5,1]}]}]
   }]
 }
 Rocket_01_Sep {
  id=#3
  possessed_class=BP_Geode_NightNight_Separated_C
  tracks [
   {type=MovieScene3DAttachTrack, sections [{type=MovieScene3DAttachSection, end_frame=3518, start_frame=1019}]}
   {
    type=MovieSceneSpawnTrack
    sections [{
      type=MovieSceneSpawnSection
      channels [None {type=MovieSceneScriptingBoolChannel, default=false, keys [[419,false],[1019,true],[1886,false]]}]
     }]
   }
   {
    type=MovieScene3DTransformTrack
    sections [{
      type=MovieScene3DTransformSection
      channels [
       Location.X {type=MovieSceneScriptingDoubleChannel, default=0}
       Location.Y {type=MovieSceneScriptingDoubleChannel, default=0}
       Location.Z {type=MovieSceneScriptingDoubleChannel, default=0}
       Rotation.X {type=MovieSceneScriptingDoubleChannel, default=0}
       Rotation.Y {type=MovieSceneScriptingDoubleChannel, default=-90}
       Rotation.Z {type=MovieSceneScriptingDoubleChannel, default=0}
       Scale.X {type=MovieSceneScriptingDoubleChannel, default=1}
       Scale.Y {type=MovieSceneScriptingDoubleChannel, default=1}
       Scale.Z {type=MovieSceneScriptingDoubleChannel, default=1}
      ]
     }]
   }
  ]
 }
 Rocket_01_Path {
  id=#4
  possessed_class=CameraRig_Rail
  tracks [
   {
    type=MovieSceneSpawnTrack
    sections [{
      type=MovieSceneSpawnSection
      channels [None {type=MovieSceneScriptingBoolChannel, default=false, keys [[419,false],[1019,true],[1886,false]]}]
     }]
   }
   {
    type=MovieScene3DTransformTrack
    sections [{
      type=MovieScene3DTransformSection
      channels [
       Location.X {type=MovieSceneScriptingDoubleChannel, default=2079.1401}
       Location.Y {type=MovieSceneScriptingDoubleChannel, default=-93415.0234}
       Location.Z {type=MovieSceneScriptingDoubleChannel, default=101008.4062}
       Rotation.X {type=MovieSceneScriptingDoubleChannel, default=0}
       Rotation.Y {type=MovieSceneScriptingDoubleChannel, default=0}
       Rotation.Z {type=MovieSceneScriptingDoubleChannel, default=72.4982}
       Scale.X {type=MovieSceneScriptingDoubleChannel, default=1}
       Scale.Y {type=MovieSceneScriptingDoubleChannel, default=1}
       Scale.Z {type=MovieSceneScriptingDoubleChannel, default=1}
      ]
     }]
   }
   {
    type=MovieSceneFloatTrack
    display_name=Current Position on Rail
    sections [{
      type=MovieSceneFloatSection
      end_frame=2945
      start_frame=823
      channels [None {type=MovieSceneScriptingFloatChannel, default=1, keys [[1019,0.12],[1441,0.2916],[1736,0.9369]]}]
     }]
   }
  ]
 }
 NightNight_SmallerRocket_CameraShakeComponent {
  id=#5
  parent_id=#6
  possessed_class=NightNight_SmallerRocket_CameraShakeComponent_C
  tracks [{
    type=MovieSceneFloatTrack
    display_name=Shake Strength
    sections [{type=MovieSceneFloatSection, channels [None {type=MovieSceneScriptingFloatChannel, default=0.5, const [0.5,1]}]}]
   }]
 }
 Rocket_07_Path {
  id=#7
  possessed_class=CameraRig_Rail
  tracks [
   {
    type=MovieSceneSpawnTrack
    sections [{
      type=MovieSceneSpawnSection
      channels [None {type=MovieSceneScriptingBoolChannel, default=false, keys [[1019,false],[2159,true],[2869.635,false]]}]
     }]
   }
   {
    type=MovieScene3DTransformTrack
    sections [{
      type=MovieScene3DTransformSection
      channels [
       Location.X {type=MovieSceneScriptingDoubleChannel, default=120683.7188}
       Location.Y {type=MovieSceneScriptingDoubleChannel, default=-85736.5469}
       Location.Z {type=MovieSceneScriptingDoubleChannel, default=98711.4375}
       Rotation.X {type=MovieSceneScriptingDoubleChannel, default=0}
       Rotation.Y {type=MovieSceneScriptingDoubleChannel, default=0}
       Rotation.Z {type=MovieSceneScriptingDoubleChannel, default=146.8914}
       Scale.X {type=MovieSceneScriptingDoubleChannel, default=1}
       Scale.Y {type=MovieSceneScriptingDoubleChannel, default=1}
       Scale.Z {type=MovieSceneScriptingDoubleChannel, default=1}
      ]
     }]
   }
   {
    type=MovieSceneFloatTrack
    display_name=Current Position on Rail
    sections [{
      type=MovieSceneFloatSection
      end_frame=4141
      start_frame=2159
      channels [None {type=MovieSceneScriptingFloatChannel, default=1, keys [[2159,0.12],[2491,0.2916],[2719.635,0.8]]}]
     }]
   }
  ]
 }
 NightNight_SmallerRocket_CameraShakeComponent {
  id=#8
  parent_id=#9
  possessed_class=NightNight_SmallerRocket_CameraShakeComponent_C
  tracks [{
    type=MovieSceneFloatTrack
    display_name=Shake Strength
    sections [{type=MovieSceneFloatSection, channels [None {type=MovieSceneScriptingFloatChannel, default=0.5, const [0.5,1]}]}]
   }]
 }
 MissileSilo_Stage2_Main {
  id=#10
  parent_id=#11
  possessed_class=StaticMeshComponent
  tracks [{
    type=MovieScene3DTransformTrack
    sections [{
      type=MovieScene3DTransformSection
      channels [
       Scale.X {type=MovieSceneScriptingDoubleChannel, default=0, keys [[2161,1],[2207,0]]}
       Scale.Y {type=MovieSceneScriptingDoubleChannel, default=0, keys [[2161,1],[2207,0]]}
       Scale.Z {type=MovieSceneScriptingDoubleChannel, default=0, keys [[2161,1],[2207,0]]}
      ]
     }]
   }]
 }
 MissileSilo_Stage2_Main {
  id=#12
  parent_id=#3
  possessed_class=StaticMeshComponent
  tracks [{
    type=MovieScene3DTransformTrack
    sections [{
      type=MovieScene3DTransformSection
      channels [
       Scale.X {type=MovieSceneScriptingDoubleChannel, default=0, keys [[1683,1],[1736,0]]}
       Scale.Y {type=MovieSceneScriptingDoubleChannel, default=0, keys [[1683,1],[1736,0]]}
       Scale.Z {type=MovieSceneScriptingDoubleChannel, default=0, keys [[1683,1],[1736,0]]}
      ]
     }]
   }]
 }
 Rocket_05_Sep {
  id=#13
  possessed_class=BP_Geode_NightNight_Separated_C
  tracks [
   {type=MovieScene3DAttachTrack, sections [{type=MovieScene3DAttachSection, end_frame=2820, start_frame=2006}]}
   {
    type=MovieSceneSpawnTrack
    sections [{
      type=MovieSceneSpawnSection
      channels [None {type=MovieSceneScriptingBoolChannel, default=false, keys [[1019,false],[2039,true],[2750,false]]}]
     }]
   }
   {
    type=MovieScene3DTransformTrack
    sections [{
      type=MovieScene3DTransformSection
      channels [
       Location.X {type=MovieSceneScriptingDoubleChannel, default=0}
       Location.Y {type=MovieSceneScriptingDoubleChannel, default=0}
       Location.Z {type=MovieSceneScriptingDoubleChannel, default=0}
       Rotation.X {type=MovieSceneScriptingDoubleChannel, default=0}
       Rotation.Y {type=MovieSceneScriptingDoubleChannel, default=-90}
       Rotation.Z {type=MovieSceneScriptingDoubleChannel, default=0}
       Scale.X {type=MovieSceneScriptingDoubleChannel, default=1}
       Scale.Y {type=MovieSceneScriptingDoubleChannel, default=1}
       Scale.Z {type=MovieSceneScriptingDoubleChannel, default=1}
      ]
     }]
   }
  ]
 }
 Rocket_02_Path {
  id=#14
  possessed_class=CameraRig_Rail
  tracks [
   {
    type=MovieSceneSpawnTrack
    sections [{
      type=MovieSceneSpawnSection
      channels [None {type=MovieSceneScriptingBoolChannel, default=false, keys [[419,false],[1619,true],[2357,false],[2745,false],[2845,false]]}]
     }]
   }
   {
    type=MovieScene3DTransformTrack
    sections [{
      type=MovieScene3DTransformSection
      channels [
       Location.X {type=MovieSceneScriptingDoubleChannel, default=33966.4727}
       Location.Y {type=MovieSceneScriptingDoubleChannel, default=-60233.4141}
       Location.Z {type=MovieSceneScriptingDoubleChannel, default=98711.4375}
       Rotation.X {type=MovieSceneScriptingDoubleChannel, default=0}
       Rotation.Y {type=MovieSceneScriptingDoubleChannel, default=0}
       Rotation.Z {type=MovieSceneScriptingDoubleChannel, default=89.9008}
       Scale.X {type=MovieSceneScriptingDoubleChannel, default=1}
       Scale.Y {type=MovieSceneScriptingDoubleChannel, default=1}
       Scale.Z {type=MovieSceneScriptingDoubleChannel, default=1}
      ]
     }]
   }
   {
    type=MovieSceneFloatTrack
    display_name=Current Position on Rail
    sections [{
      type=MovieSceneFloatSection
      end_frame=2945
      start_frame=1019
      channels [None {type=MovieSceneScriptingFloatChannel, default=1, keys [[1619,0.12],[1951,0.2916],[2207,0.7947]]}]
     }]
   }
  ]
 }
 NightNight_SmallerRocket_CameraShakeComponent {
  id=#15
  parent_id=#11
  possessed_class=NightNight_SmallerRocket_CameraShakeComponent_C
  tracks [{
    type=MovieSceneFloatTrack
    display_name=Shake Strength
    sections [{type=MovieSceneFloatSection, channels [None {type=MovieSceneScriptingFloatChannel, default=0.5, const [0.5,1]}]}]
   }]
 }
 MissileSilo_Stage2_Main {
  id=#16
  parent_id=#13
  possessed_class=StaticMeshComponent
  tracks [{
    type=MovieScene3DTransformTrack
    sections [{
      type=MovieScene3DTransformSection
      channels [
       Scale.X {type=MovieSceneScriptingDoubleChannel, default=0, keys [[2509,1],[2600,0]]}
       Scale.Y {type=MovieSceneScriptingDoubleChannel, default=0, keys [[2509,1],[2600,0]]}
       Scale.Z {type=MovieSceneScriptingDoubleChannel, default=0, keys [[2509,1],[2600,0]]}
      ]
     }]
   }]
 }
 Rocket_03_Sep {
  id=#9
  possessed_class=BP_Geode_NightNight_Separated_C
  tracks [
   {type=MovieScene3DAttachTrack, sections [{type=MovieScene3DAttachSection, end_frame=3358, start_frame=1196}]}
   {
    type=MovieSceneSpawnTrack
    sections [{
      type=MovieSceneSpawnSection
      end_frame=2933
      start_frame=1619
      channels [None {type=MovieSceneScriptingBoolChannel, default=false, keys [[137.05,false],[1919,true],[2660,false]]}]
     }]
   }
   {
    type=MovieScene3DTransformTrack
    sections [{
      type=MovieScene3DTransformSection
      channels [
       Location.X {type=MovieSceneScriptingDoubleChannel, default=0}
       Location.Y {type=MovieSceneScriptingDoubleChannel, default=0}
       Location.Z {type=MovieSceneScriptingDoubleChannel, default=0}
       Rotation.X {type=MovieSceneScriptingDoubleChannel, default=0}
       Rotation.Y {type=MovieSceneScriptingDoubleChannel, default=-90}
       Rotation.Z {type=MovieSceneScriptingDoubleChannel, default=0}
       Scale.X {type=MovieSceneScriptingDoubleChannel, default=1}
       Scale.Y {type=MovieSceneScriptingDoubleChannel, default=1}
       Scale.Z {type=MovieSceneScriptingDoubleChannel, default=1}
      ]
     }]
   }
  ]
 }
 NightNight_SmallerRocket_CameraShakeComponent {
  id=#17
  parent_id=#3
  possessed_class=NightNight_SmallerRocket_CameraShakeComponent_C
  tracks [{
    type=MovieSceneFloatTrack
    display_name=Shake Strength
    sections [{type=MovieSceneFloatSection, channels [None {type=MovieSceneScriptingFloatChannel, default=0.5, const [0.5,1]}]}]
   }]
 }
 MissileSilo_Stage2_Main {
  id=#18
  parent_id=#9
  possessed_class=StaticMeshComponent
  tracks [{
    type=MovieScene3DTransformTrack
    sections [{
      type=MovieScene3DTransformSection
      channels [
       Scale.X {type=MovieSceneScriptingDoubleChannel, default=0, keys [[2487,1],[2510,0]]}
       Scale.Y {type=MovieSceneScriptingDoubleChannel, default=0, keys [[2487,1],[2510,0]]}
       Scale.Z {type=MovieSceneScriptingDoubleChannel, default=0, keys [[2487,1],[2510,0]]}
      ]
     }]
   }]
 }
 Rocket_07_Sep {
  id=#2
  possessed_class=BP_Geode_NightNight_Separated_C
  tracks [
   {type=MovieScene3DAttachTrack, sections [{type=MovieScene3DAttachSection, end_frame=3127, start_frame=1019}]}
   {
    type=MovieSceneSpawnTrack
    sections [{
      type=MovieSceneSpawnSection
      channels [None {type=MovieSceneScriptingBoolChannel, default=false, keys [[1019,false],[2159,true],[2869.635,false],[3419,false]]}]
     }]
   }
   {
    type=MovieScene3DTransformTrack
    sections [{
      type=MovieScene3DTransformSection
      channels [
       Location.X {type=MovieSceneScriptingDoubleChannel, default=0}
       Location.Y {type=MovieSceneScriptingDoubleChannel, default=0}
       Location.Z {type=MovieSceneScriptingDoubleChannel, default=0}
       Rotation.X {type=MovieSceneScriptingDoubleChannel, default=0}
       Rotation.Y {type=MovieSceneScriptingDoubleChannel, default=-90}
       Rotation.Z {type=MovieSceneScriptingDoubleChannel, default=0}
       Scale.X {type=MovieSceneScriptingDoubleChannel, default=1}
       Scale.Y {type=MovieSceneScriptingDoubleChannel, default=1}
       Scale.Z {type=MovieSceneScriptingDoubleChannel, default=1}
      ]
     }]
   }
  ]
 }
 NightNight_SmallerRocket_CameraShakeComponent {
  id=#19
  parent_id=#13
  possessed_class=NightNight_SmallerRocket_CameraShakeComponent_C
  tracks [{
    type=MovieSceneFloatTrack
    display_name=Shake Strength
    sections [{type=MovieSceneFloatSection, channels [None {type=MovieSceneScriptingFloatChannel, default=0.5, const [0.5,1]}]}]
   }]
 }
 MissileSilo_Stage2_Main {
  id=#20
  parent_id=#2
  possessed_class=StaticMeshComponent
  tracks [{
    type=MovieScene3DTransformTrack
    sections [{
      type=MovieScene3DTransformSection
      channels [
       Scale.X {type=MovieSceneScriptingDoubleChannel, default=0, keys [[2588,1],[2719.635,0]]}
       Scale.Y {type=MovieSceneScriptingDoubleChannel, default=0, keys [[2588,1],[2719.635,0]]}
       Scale.Z {type=MovieSceneScriptingDoubleChannel, default=0, keys [[2588,1],[2719.635,0]]}
      ]
     }]
   }]
 }
 MissileSilo_Stage2_Main {
  id=#21
  parent_id=#22
  possessed_class=StaticMeshComponent
  tracks [{
    type=MovieScene3DTransformTrack
    sections [{
      type=MovieScene3DTransformSection
      channels [
       Scale.X {type=MovieSceneScriptingDoubleChannel, default=0, keys [[2473,1],[2549.635,0]]}
       Scale.Y {type=MovieSceneScriptingDoubleChannel, default=0, keys [[2473,1],[2549.635,0]]}
       Scale.Z {type=MovieSceneScriptingDoubleChannel, default=0, keys [[2473,1],[2549.635,0]]}
      ]
     }]
   }]
 }
 Rocket_03_Path {
  id=#23
  possessed_class=CameraRig_Rail
  tracks [
   {
    type=MovieSceneSpawnTrack
    sections [{
      type=MovieSceneSpawnSection
      channels [None {type=MovieSceneScriptingBoolChannel, default=false, keys [[1019,false],[1919,true],[2660,false]]}]
     }]
   }
   {
    type=MovieScene3DTransformTrack
    sections [{
      type=MovieScene3DTransformSection
      channels [
       Location.X {type=MovieSceneScriptingDoubleChannel, default=22933.5273}
       Location.Y {type=MovieSceneScriptingDoubleChannel, default=38174.4844}
       Location.Z {type=MovieSceneScriptingDoubleChannel, default=98711.4375}
       Rotation.X {type=MovieSceneScriptingDoubleChannel, default=0}
       Rotation.Y {type=MovieSceneScriptingDoubleChannel, default=0}
       Rotation.Z {type=MovieSceneScriptingDoubleChannel, default=-80.5409}
       Scale.X {type=MovieSceneScriptingDoubleChannel, default=1}
       Scale.Y {type=MovieSceneScriptingDoubleChannel, default=1}
       Scale.Z {type=MovieSceneScriptingDoubleChannel, default=1}
      ]
     }]
   }
   {
    type=MovieSceneFloatTrack
    display_name=Current Position on Rail
    sections [{
      type=MovieSceneFloatSection
      end_frame=3181
      start_frame=1619
      channels [None {type=MovieSceneScriptingFloatChannel, default=1, keys [[1919,0.12],[2239,0.3255],[2461,0.6521],[2510,0.7607]]}]
     }]
   }
  ]
 }
 NightNight_SmallerRocket_CameraShakeComponent {
  id=#24
  parent_id=#22
  possessed_class=NightNight_SmallerRocket_CameraShakeComponent_C
  tracks [{
    type=MovieSceneFloatTrack
    display_name=Shake Strength
    sections [{type=MovieSceneFloatSection, channels [None {type=MovieSceneScriptingFloatChannel, default=0.5, const [0.5,1]}]}]
   }]
 }
 Rocket_04_Sep {
  id=#22
  possessed_class=BP_Geode_NightNight_Separated_C
  tracks [
   {type=MovieScene3DAttachTrack, sections [{type=MovieScene3DAttachSection, end_frame=3246, start_frame=1019}]}
   {
    type=MovieSceneSpawnTrack
    sections [{
      type=MovieSceneSpawnSection
      channels [None {type=MovieSceneScriptingBoolChannel, default=false, keys [[1019,false],[1979,true],[2699.635,false],[3546.0025,false]]}]
     }]
   }
   {
    type=MovieScene3DTransformTrack
    sections [{
      type=MovieScene3DTransformSection
      channels [
       Location.X {type=MovieSceneScriptingDoubleChannel, default=0}
       Location.Y {type=MovieSceneScriptingDoubleChannel, default=0}
       Location.Z {type=MovieSceneScriptingDoubleChannel, default=0}
       Rotation.X {type=MovieSceneScriptingDoubleChannel, default=0}
       Rotation.Y {type=MovieSceneScriptingDoubleChannel, default=-90}
       Rotation.Z {type=MovieSceneScriptingDoubleChannel, default=0}
       Scale.X {type=MovieSceneScriptingDoubleChannel, default=1}
       Scale.Y {type=MovieSceneScriptingDoubleChannel, default=1}
       Scale.Z {type=MovieSceneScriptingDoubleChannel, default=1}
      ]
     }]
   }
  ]
 }
 MissileSilo_Stage2_Main {
  id=#25
  parent_id=#6
  possessed_class=StaticMeshComponent
  tracks [{
    type=MovieScene3DTransformTrack
    sections [{
      type=MovieScene3DTransformSection
      channels [
       Scale.X {type=MovieSceneScriptingDoubleChannel, default=0, keys [[2600,1],[2680.635,0]]}
       Scale.Y {type=MovieSceneScriptingDoubleChannel, default=0, keys [[2600,1],[2680.635,0]]}
       Scale.Z {type=MovieSceneScriptingDoubleChannel, default=0, keys [[2600,1],[2680.635,0]]}
      ]
     }]
   }]
 }
 Rocket_04_Path {
  id=#26
  possessed_class=CameraRig_Rail
  tracks [
   {
    type=MovieSceneSpawnTrack
    sections [{type=MovieSceneSpawnSection, channels [None {type=MovieSceneScriptingBoolChannel, default=false, keys [[1979,true],[2699.635,false]]}]}]
   }
   {
    type=MovieScene3DTransformTrack
    sections [{
      type=MovieScene3DTransformSection
      channels [
       Location.X {type=MovieSceneScriptingDoubleChannel, default=-28881.2969}
       Location.Y {type=MovieSceneScriptingDoubleChannel, default=58283.2344}
       Location.Z {type=MovieSceneScriptingDoubleChannel, default=98711.4375}
       Rotation.X {type=MovieSceneScriptingDoubleChannel, default=0}
       Rotation.Y {type=MovieSceneScriptingDoubleChannel, default=0}
       Rotation.Z {type=MovieSceneScriptingDoubleChannel, default=321}
       Scale.X {type=MovieSceneScriptingDoubleChannel, default=1}
       Scale.Y {type=MovieSceneScriptingDoubleChannel, default=1}
       Scale.Z {type=MovieSceneScriptingDoubleChannel, default=1}
      ]
     }]
   }
   {
    type=MovieSceneFloatTrack
    display_name=Current Position on Rail
    sections [{
      type=MovieSceneFloatSection
      end_frame=3601
      start_frame=1979
      channels [None {type=MovieSceneScriptingFloatChannel, default=1, keys [[1979,0.12],[2311,0.2916],[2549.635,0.7595]]}]
     }]
   }
  ]
 }
 Rocket_06_Path {
  id=#27
  possessed_class=CameraRig_Rail
  tracks [
   {
    type=MovieSceneSpawnTrack
    sections [{
      type=MovieSceneSpawnSection
      channels [None {type=MovieSceneScriptingBoolChannel, default=false, keys [[1019,false],[2099,true],[2830.635,false]]}]
     }]
   }
   {
    type=MovieScene3DTransformTrack
    sections [{
      type=MovieScene3DTransformSection
      channels [
       Location.X {type=MovieSceneScriptingDoubleChannel, default=-56560.625}
       Location.Y {type=MovieSceneScriptingDoubleChannel, default=29723.0938}
       Location.Z {type=MovieSceneScriptingDoubleChannel, default=101321.8594}
       Rotation.X {type=MovieSceneScriptingDoubleChannel, default=0}
       Rotation.Y {type=MovieSceneScriptingDoubleChannel, default=0}
       Rotation.Z {type=MovieSceneScriptingDoubleChannel, default=-26.0164}
       Scale.X {type=MovieSceneScriptingDoubleChannel, default=1}
       Scale.Y {type=MovieSceneScriptingDoubleChannel, default=1}
       Scale.Z {type=MovieSceneScriptingDoubleChannel, default=1}
      ]
     }]
   }
   {
    type=MovieSceneFloatTrack
    display_name=Current Position on Rail
    sections [{
      type=MovieSceneFloatSection
      end_frame=3961
      start_frame=2099
      channels [None {type=MovieSceneScriptingFloatChannel, default=1, keys [[2099,0.12],[2385,0.3173],[2680.635,0.8215]]}]
     }]
   }
  ]
 }
 Rocket_06_Sep {
  id=#6
  possessed_class=BP_Geode_NightNight_Separated_C
  tracks [
   {type=MovieScene3DAttachTrack, sections [{type=MovieScene3DAttachSection, end_frame=3613, start_frame=1017}]}
   {
    type=MovieSceneSpawnTrack
    sections [{
      type=MovieSceneSpawnSection
      channels [None {type=MovieSceneScriptingBoolChannel, default=false, keys [[1019,false],[2099,true],[2830.635,false]]}]
     }]
   }
   {
    type=MovieScene3DTransformTrack
    sections [{
      type=MovieScene3DTransformSection
      channels [
       Location.X {type=MovieSceneScriptingDoubleChannel, default=0}
       Location.Y {type=MovieSceneScriptingDoubleChannel, default=0}
       Location.Z {type=MovieSceneScriptingDoubleChannel, default=0}
       Rotation.X {type=MovieSceneScriptingDoubleChannel, default=0}
       Rotation.Y {type=MovieSceneScriptingDoubleChannel, default=-90}
       Rotation.Z {type=MovieSceneScriptingDoubleChannel, default=0}
       Scale.X {type=MovieSceneScriptingDoubleChannel, default=1}
       Scale.Y {type=MovieSceneScriptingDoubleChannel, default=1}
       Scale.Z {type=MovieSceneScriptingDoubleChannel, default=1}
      ]
     }]
   }
  ]
 }
 Rocket_02_Sep {
  id=#11
  possessed_class=BP_Geode_NightNight_Separated_C
  tracks [
   {type=MovieScene3DAttachTrack, sections [{type=MovieScene3DAttachSection, end_frame=3518, start_frame=1019}]}
   {
    type=MovieSceneSpawnTrack
    sections [{
      type=MovieSceneSpawnSection
      channels [None {type=MovieSceneScriptingBoolChannel, default=false, keys [[1019,false],[1619,true],[2357,false],[2945.0025,false]]}]
     }]
   }
   {
    type=MovieScene3DTransformTrack
    sections [{
      type=MovieScene3DTransformSection
      channels [
       Location.X {type=MovieSceneScriptingDoubleChannel, default=0}
       Location.Y {type=MovieSceneScriptingDoubleChannel, default=0}
       Location.Z {type=MovieSceneScriptingDoubleChannel, default=0}
       Rotation.X {type=MovieSceneScriptingDoubleChannel, default=0}
       Rotation.Y {type=MovieSceneScriptingDoubleChannel, default=-90}
       Rotation.Z {type=MovieSceneScriptingDoubleChannel, default=0}
       Scale.X {type=MovieSceneScriptingDoubleChannel, default=1}
       Scale.Y {type=MovieSceneScriptingDoubleChannel, default=1}
       Scale.Z {type=MovieSceneScriptingDoubleChannel, default=1}
      ]
     }]
   }
  ]
 }
 Rocket_05_Path {
  id=#28
  possessed_class=CameraRig_Rail
  tracks [
   {
    type=MovieSceneSpawnTrack
    sections [{
      type=MovieSceneSpawnSection
      channels [None {type=MovieSceneScriptingBoolChannel, default=false, keys [[1019,false],[2039,true],[2750,false]]}]
     }]
   }
   {
    type=MovieScene3DTransformTrack
    sections [{
      type=MovieScene3DTransformSection
      channels [
       Location.X {type=MovieSceneScriptingDoubleChannel, default=-32484.4941}
       Location.Y {type=MovieSceneScriptingDoubleChannel, default=532.7101}
       Location.Z {type=MovieSceneScriptingDoubleChannel, default=101983.0312}
       Rotation.X {type=MovieSceneScriptingDoubleChannel, default=0}
       Rotation.Y {type=MovieSceneScriptingDoubleChannel, default=0}
       Rotation.Z {type=MovieSceneScriptingDoubleChannel, default=340.2012}
       Scale.X {type=MovieSceneScriptingDoubleChannel, default=1}
       Scale.Y {type=MovieSceneScriptingDoubleChannel, default=1}
       Scale.Z {type=MovieSceneScriptingDoubleChannel, default=1}
      ]
     }]
   }
   {
    type=MovieSceneFloatTrack
    display_name=Current Position on Rail
    sections [{
      type=MovieSceneFloatSection
      end_frame=3781
      start_frame=2039
      channels [None {type=MovieSceneScriptingFloatChannel, default=1, keys [[2039,0.12],[2371,0.2916],[2600,0.7375]]}]
     }]
   }
  ]
 }
]
display_rate {denominator=1, numerator=60}
tick_resolution {denominator=1, numerator=24000}
tracks [{
  type=MovieSceneEventTrack
  sections [{
    type=MovieSceneEventTriggerSection
    channels [None {type=MovieSceneScriptingEventChannel, const ["<Struct 'MovieSceneEvent' (0x49e50ac9100) {}>", 1]}]
   }]
 }]
