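# Level Sequence for part 2 of the NightNight effects, staging rifts that open and close at fixed points across the map.
# Spawn tracks toggle seven Rift Enter emitters and six high-altitude, vertically pitched RiftExit_Long Blueprints in staggered, overlapping windows, each at a static transform.
# Particle tracks trigger each entry emitter and the RiftExit particle component nested under each exit Blueprint.
# A scaled SpatialMarker static mesh stays spawned throughout; there are no master tracks.
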LevelSequence /Figment_S10/Athena/Prototype/Blueprints/NightNight/NightNight_Part2_FX
bindings [
 Rift Enter {
  id=#1
  possessed_class=Emitter
  tracks [
   {
    type=MovieSceneSpawnTrack
    sections [{
      type=MovieSceneSpawnSection
      end_frame=3001
      start_frame=2351
      channels [None {type=MovieSceneScriptingBoolChannel, default=false, keys [[2351,false],[2640,true],[3001,false]]}]
     }]
   }
   {
    type=MovieScene3DTransformTrack
    sections [{
      type=MovieScene3DTransformSection
      channels [
       Location.X {type=MovieSceneScriptingDoubleChannel, default=68635.4766}
       Location.Y {type=MovieSceneScriptingDoubleChannel, default=-7358.708}
       Location.Z {type=MovieSceneScriptingDoubleChannel, default=24812.9004}
       Rotation.X {type=MovieSceneScriptingDoubleChannel, default=3.5675}
       Rotation.Y {type=MovieSceneScriptingDoubleChannel, default=16.848}
       Rotation.Z {type=MovieSceneScriptingDoubleChannel, default=29.2068}
       Scale.X {type=MovieSceneScriptingDoubleChannel, default=1}
       Scale.Y {type=MovieSceneScriptingDoubleChannel, default=1}
       Scale.Z {type=MovieSceneScriptingDoubleChannel, default=1}
      ]
     }]
   }
   {
    type=MovieSceneParticleTrack
    display_name=Particle System
    sections [{type=MovieSceneParticleSection, channels [None {type=MovieSceneScriptingParticleChannel, const ["<ParticleKey.TRIGGER: 2>", 1]}]}]
   }
  ]
 }
 RiftExit {
  id=#2
  parent_id=#3
  possessed_class=ParticleSystemComponent
  tracks [{
    type=MovieSceneParticleTrack
    display_name=Particle System
    sections [{type=MovieSceneParticleSection, channels [None {type=MovieSceneScriptingParticleChannel, const ["<ParticleKey.TRIGGER: 2>", 1]}]}]
   }]
 }
 Rift Enter {
  id=#4
  possessed_class=Emitter
  tracks [
   {
    type=MovieSceneSpawnTrack
    sections [{
      type=MovieSceneSpawnSection
      end_frame=2021
      start_frame=1423
      channels [None {type=MovieSceneScriptingBoolChannel, default=false, keys [[1423,false],[1695.6,true],[2021,false]]}]
     }]
   }
   {
    type=MovieScene3DTransformTrack
    sections [{
      type=MovieScene3DTransformSection
      channels [
       Location.X {type=MovieSceneScriptingDoubleChannel, default=48097.7109}
       Location.Y {type=MovieSceneScriptingDoubleChannel, default=52678.9805}
       Location.Z {type=MovieSceneScriptingDoubleChannel, default=9763.665}
       Rotation.X {type=MovieSceneScriptingDoubleChannel, default=0}
       Rotation.Y {type=MovieSceneScriptingDoubleChannel, default=0}
       Rotation.Z {type=MovieSceneScriptingDoubleChannel, default=71.6925}
       Scale.X {type=MovieSceneScriptingDoubleChannel, default=1}
       Scale.Y {type=MovieSceneScriptingDoubleChannel, default=1}
       Scale.Z {type=MovieSceneScriptingDoubleChannel, default=1}
      ]
     }]
   }
   {
    type=MovieSceneParticleTrack
    display_name=Particle System
    sections [{type=MovieSceneParticleSection, channels [None {type=MovieSceneScriptingParticleChannel, const ["<ParticleKey.TRIGGER: 2>", 1]}]}]
   }
  ]
 }
 RiftExit {
  id=#5
  parent_id=#6
  possessed_class=ParticleSystemComponent
  tracks [{
    type=MovieSceneParticleTrack
    display_name=Particle System
    sections [{type=MovieSceneParticleSection, channels [None {type=MovieSceneScriptingParticleChannel, const ["<ParticleKey.TRIGGER: 2>", 1]}]}]
   }]
 }
 BP_NightNight_Generic_Rift_Exit_Long {
  id=#7
  possessed_class=BP_NightNight_Generic_RiftExit_Long_C
  tracks [
   {
    type=MovieSceneSpawnTrack
    sections [{
      type=MovieSceneSpawnSection
      end_frame=2539
      start_frame=1533
      channels [None {type=MovieSceneScriptingBoolChannel, default=false, keys [[1533,false],[1738.8,true],[2539,false]]}]
     }]
   }
   {
    type=MovieScene3DTransformTrack
    sections [{
      type=MovieScene3DTransformSection
      channels [
       Location.X {type=MovieSceneScriptingDoubleChannel, default=-56560.6094}
       Location.Y {type=MovieSceneScriptingDoubleChannel, default=29723.1836}
       Location.Z {type=MovieSceneScriptingDoubleChannel, default=102143.1875}
       Rotation.X {type=MovieSceneScriptingDoubleChannel, default=21.3405}
       Rotation.Y {type=MovieSceneScriptingDoubleChannel, default=-90}
       Rotation.Z {type=MovieSceneScriptingDoubleChannel, default=0}
       Scale.X {type=MovieSceneScriptingDoubleChannel, default=1}
       Scale.Y {type=MovieSceneScriptingDoubleChannel, default=1}
       Scale.Z {type=MovieSceneScriptingDoubleChannel, default=1}
      ]
     }]
   }
  ]
 }
 Rift Enter {
  id=#8
  possessed_class=Emitter
  tracks [
   {
    type=MovieSceneSpawnTrack
    sections [{
      type=MovieSceneSpawnSection
      end_frame=3083
      start_frame=2159
      channels [None {type=MovieSceneScriptingBoolChannel, default=false, keys [[2159,false],[2679,true],[3083,false]]}]
     }]
   }
   {
    type=MovieScene3DTransformTrack
    sections [{
      type=MovieScene3DTransformSection
      channels [
       Location.X {type=MovieSceneScriptingDoubleChannel, default=7026.5947}
       Location.Y {type=MovieSceneScriptingDoubleChannel, default=-11621.375}
       Location.Z {type=MovieSceneScriptingDoubleChannel, default=11712.5645}
       Rotation.X {type=MovieSceneScriptingDoubleChannel, default=0}
       Rotation.Y {type=MovieSceneScriptingDoubleChannel, default=0}
       Rotation.Z {type=MovieSceneScriptingDoubleChannel, default=147.0858}
       Scale.X {type=MovieSceneScriptingDoubleChannel, default=1}
       Scale.Y {type=MovieSceneScriptingDoubleChannel, default=1}
       Scale.Z {type=MovieSceneScriptingDoubleChannel, default=1}
      ]
     }]
   }
   {
    type=MovieSceneParticleTrack
    display_name=Particle System
    sections [{type=MovieSceneParticleSection, channels [None {type=MovieSceneScriptingParticleChannel, const ["<ParticleKey.TRIGGER: 2>", 1]}]}]
   }
  ]
 }
 Rift Enter {
  id=#9
  possessed_class=Emitter
  tracks [
   {
    type=MovieSceneSpawnTrack
    sections [{
      type=MovieSceneSpawnSection
      end_frame=2855
      start_frame=1968
      channels [None {type=MovieSceneScriptingBoolChannel, default=false, keys [[2159,false],[2509,true],[2855,false]]}]
     }]
   }
   {
    type=MovieScene3DTransformTrack
    sections [{
      type=MovieScene3DTransformSection
      channels [
       Location.X {type=MovieSceneScriptingDoubleChannel, default=52903.3867}
       Location.Y {type=MovieSceneScriptingDoubleChannel, default=-7925.4189}
       Location.Z {type=MovieSceneScriptingDoubleChannel, default=8802.7344}
       Rotation.X {type=MovieSceneScriptingDoubleChannel, default=0}
       Rotation.Y {type=MovieSceneScriptingDoubleChannel, default=0}
       Rotation.Z {type=MovieSceneScriptingDoubleChannel, default=-37.1479}
       Scale.X {type=MovieSceneScriptingDoubleChannel, default=1}
       Scale.Y {type=MovieSceneScriptingDoubleChannel, default=1}
       Scale.Z {type=MovieSceneScriptingDoubleChannel, default=1}
      ]
     }]
   }
   {
    type=MovieSceneParticleTrack
    display_name=Particle System
    sections [{type=MovieSceneParticleSection, channels [None {type=MovieSceneScriptingParticleChannel, const ["<ParticleKey.TRIGGER: 2>", 1]}]}]
   }
  ]
 }
 RiftExit {
  id=#10
  parent_id=#11
  possessed_class=ParticleSystemComponent
  tracks [{
    type=MovieSceneParticleTrack
    display_name=Particle System
    sections [{type=MovieSceneParticleSection, channels [None {type=MovieSceneScriptingParticleChannel, const ["<ParticleKey.TRIGGER: 2>", 1]}]}]
   }]
 }
 BP_NightNight_Generic_Rift_Exit_Long {
  id=#12
  possessed_class=BP_NightNight_Generic_RiftExit_Long_C
  tracks [
   {
    type=MovieSceneSpawnTrack
    sections [{
      type=MovieSceneSpawnSection
      channels [None {type=MovieSceneScriptingBoolChannel, default=false, keys [[1019,false],[1558.8,true],[2359,false]]}]
     }]
   }
   {
    type=MovieScene3DTransformTrack
    sections [{
      type=MovieScene3DTransformSection
      channels [
       Location.X {type=MovieSceneScriptingDoubleChannel, default=22935.998}
       Location.Y {type=MovieSceneScriptingDoubleChannel, default=38174}
       Location.Z {type=MovieSceneScriptingDoubleChannel, default=101856}
       Rotation.X {type=MovieSceneScriptingDoubleChannel, default=127.7726}
       Rotation.Y {type=MovieSceneScriptingDoubleChannel, default=-90}
       Rotation.Z {type=MovieSceneScriptingDoubleChannel, default=7.125}
       Scale.X {type=MovieSceneScriptingDoubleChannel, default=1}
       Scale.Y {type=MovieSceneScriptingDoubleChannel, default=1}
       Scale.Z {type=MovieSceneScriptingDoubleChannel, default=1}
      ]
     }]
   }
  ]
 }
 RiftEnter {
  id=#13
  possessed_class=Emitter
  tracks [
   {
    type=MovieSceneSpawnTrack
    sections [{
      type=MovieSceneSpawnSection
      end_frame=2928
      start_frame=2159
      channels [None {type=MovieSceneScriptingBoolChannel, default=false, keys [[2159,false],[2559,true],[2928,false]]}]
     }]
   }
   {
    type=MovieScene3DTransformTrack
    sections [{
      type=MovieScene3DTransformSection
      channels [
       Location.X {type=MovieSceneScriptingDoubleChannel, default=62051.793}
       Location.Y {type=MovieSceneScriptingDoubleChannel, default=-33262.4766}
       Location.Z {type=MovieSceneScriptingDoubleChannel, default=15969.6836}
       Rotation.X {type=MovieSceneScriptingDoubleChannel, default=0}
       Rotation.Y {type=MovieSceneScriptingDoubleChannel, default=0}
       Rotation.Z {type=MovieSceneScriptingDoubleChannel, default=-21.3071}
       Scale.X {type=MovieSceneScriptingDoubleChannel, default=1}
       Scale.Y {type=MovieSceneScriptingDoubleChannel, default=1}
       Scale.Z {type=MovieSceneScriptingDoubleChannel, default=1}
      ]
     }]
   }
   {
    type=MovieSceneParticleTrack
    display_name=Particle System
    sections [{type=MovieSceneParticleSection, channels [None {type=MovieSceneScriptingParticleChannel, const ["<ParticleKey.TRIGGER: 2>", 1]}]}]
   }
  ]
 }
 BP_NightNight_Generic_Rift_Exit_Long {
  id=#11
  possessed_class=BP_NightNight_Generic_RiftExit_Long_C
  tracks [
   {
    type=MovieSceneSpawnTrack
    sections [{
      type=MovieSceneSpawnSection
      end_frame=2059
      start_frame=1019
      channels [None {type=MovieSceneScriptingBoolChannel, default=false, keys [[1019,false],[1258.8,true],[2059,false]]}]
     }]
   }
   {
    type=MovieScene3DTransformTrack
    sections [{
      type=MovieScene3DTransformSection
      channels [
       Location.X {type=MovieSceneScriptingDoubleChannel, default=33986.6875}
       Location.Y {type=MovieSceneScriptingDoubleChannel, default=-60285.3789}
       Location.Z {type=MovieSceneScriptingDoubleChannel, default=101098}
       Rotation.X {type=MovieSceneScriptingDoubleChannel, default=0}
       Rotation.Y {type=MovieSceneScriptingDoubleChannel, default=-90}
       Rotation.Z {type=MovieSceneScriptingDoubleChannel, default=0}
       Scale.X {type=MovieSceneScriptingDoubleChannel, default=1}
       Scale.Y {type=MovieSceneScriptingDoubleChannel, default=1}
       Scale.Z {type=MovieSceneScriptingDoubleChannel, default=1}
      ]
     }]
   }
  ]
 }
 RiftExit {
  id=#14
  parent_id=#7
  possessed_class=ParticleSystemComponent
  tracks [{
    type=MovieSceneParticleTrack
    display_name=Particle System
    sections [{type=MovieSceneParticleSection, channels [None {type=MovieSceneScriptingParticleChannel, const ["<ParticleKey.TRIGGER: 2>", 1]}]}]
   }]
 }
 Rift Enter {
  id=#15
  possessed_class=Emitter
  tracks [
   {
    type=MovieSceneSpawnTrack
    sections [{
      type=MovieSceneSpawnSection
      end_frame=2791
      start_frame=2039
      channels [None {type=MovieSceneScriptingBoolChannel, default=false, keys [[2039,false],[2470,true],[2791,false]]}]
     }]
   }
   {
    type=MovieScene3DTransformTrack
    sections [{
      type=MovieScene3DTransformSection
      channels [
       Location.X {type=MovieSceneScriptingDoubleChannel, default=12848.0254}
       Location.Y {type=MovieSceneScriptingDoubleChannel, default=-70309.7812}
       Location.Z {type=MovieSceneScriptingDoubleChannel, default=9073.7217}
       Rotation.X {type=MovieSceneScriptingDoubleChannel, default=0}
       Rotation.Y {type=MovieSceneScriptingDoubleChannel, default=-4.0715}
       Rotation.Z {type=MovieSceneScriptingDoubleChannel, default=-145.3655}
       Scale.X {type=MovieSceneScriptingDoubleChannel, default=1}
       Scale.Y {type=MovieSceneScriptingDoubleChannel, default=1}
       Scale.Z {type=MovieSceneScriptingDoubleChannel, default=1}
      ]
     }]
   }
   {
    type=MovieSceneParticleTrack
    display_name=Particle System
    sections [{type=MovieSceneParticleSection, channels [None {type=MovieSceneScriptingParticleChannel, const ["<ParticleKey.TRIGGER: 2>", 1]}]}]
   }
  ]
 }
 RiftExit {
  id=#16
  parent_id=#17
  possessed_class=ParticleSystemComponent
  tracks [{
    type=MovieSceneParticleTrack
    display_name=Particle System
    sections [{type=MovieSceneParticleSection, channels [None {type=MovieSceneScriptingParticleChannel, const ["<ParticleKey.TRIGGER: 2>", 1]}]}]
   }]
 }
 Rift Enter {
  id=#18
  possessed_class=Emitter
  tracks [
   {
    type=MovieSceneSpawnTrack
    sections [{
      type=MovieSceneSpawnSection
      end_frame=2509
      start_frame=1785
      channels [None {type=MovieSceneScriptingBoolChannel, default=false, keys [[1785,false],[2167,true],[2509,false]]}]
     }]
   }
   {
    type=MovieScene3DTransformTrack
    sections [{
      type=MovieScene3DTransformSection
      channels [
       Location.X {type=MovieSceneScriptingDoubleChannel, default=35722.7734}
       Location.Y {type=MovieSceneScriptingDoubleChannel, default=55252.6172}
       Location.Z {type=MovieSceneScriptingDoubleChannel, default=27067.4434}
       Rotation.X {type=MovieSceneScriptingDoubleChannel, default=-0.5113}
       Rotation.Y {type=MovieSceneScriptingDoubleChannel, default=22.7213}
       Rotation.Z {type=MovieSceneScriptingDoubleChannel, default=449.2813}
       Scale.X {type=MovieSceneScriptingDoubleChannel, default=1}
       Scale.Y {type=MovieSceneScriptingDoubleChannel, default=1}
       Scale.Z {type=MovieSceneScriptingDoubleChannel, default=1}
      ]
     }]
   }
   {
    type=MovieSceneParticleTrack
    display_name=Particle System
    sections [{type=MovieSceneParticleSection, channels [None {type=MovieSceneScriptingParticleChannel, const ["<ParticleKey.TRIGGER: 2>", 1]}]}]
   }
  ]
 }
 SpatialMarker {
  id=#19
  possessed_class=FortStaticMeshActor
  tracks [
   {
    type=MovieScene3DTransformTrack
    sections [{
      type=MovieScene3DTransformSection
      channels [
       Location.X {type=MovieSceneScriptingDoubleChannel, default=36408.2695}
       Location.Y {type=MovieSceneScriptingDoubleChannel, default=-30834.1055}
       Location.Z {type=MovieSceneScriptingDoubleChannel, default=12600.5645}
       Rotation.X {type=MovieSceneScriptingDoubleChannel, default=0}
       Rotation.Y {type=MovieSceneScriptingDoubleChannel, default=0}
       Rotation.Z {type=MovieSceneScriptingDoubleChannel, default=0}
       Scale.X {type=MovieSceneScriptingDoubleChannel, default=11}
       Scale.Y {type=MovieSceneScriptingDoubleChannel, default=11}
       Scale.Z {type=MovieSceneScriptingDoubleChannel, default=11}
      ]
     }]
   }
   {type=MovieSceneSpawnTrack, sections [{type=MovieSceneSpawnSection, channels [None {type=MovieSceneScriptingBoolChannel, default=true}]}]}
  ]
 }
 BP_NightNight_Generic_Rift_Exit_Long {
  id=#17
  possessed_class=BP_NightNight_Generic_RiftExit_Long_C
  tracks [
   {
    type=MovieSceneSpawnTrack
    sections [{
      type=MovieSceneSpawnSection
      end_frame=2479
      start_frame=1533
      channels [None {type=MovieSceneScriptingBoolChannel, default=false, keys [[1533,false],[1678.8,true],[2479,false]]}]
     }]
   }
   {
    type=MovieScene3DTransformTrack
    sections [{
      type=MovieScene3DTransformSection
      channels [
       Location.X {type=MovieSceneScriptingDoubleChannel, default=-32512}
       Location.Y {type=MovieSceneScriptingDoubleChannel, default=512}
       Location.Z {type=MovieSceneScriptingDoubleChannel, default=101632}
       Rotation.X {type=MovieSceneScriptingDoubleChannel, default=89.8541}
       Rotation.Y {type=MovieSceneScriptingDoubleChannel, default=-90}
       Rotation.Z {type=MovieSceneScriptingDoubleChannel, default=-26.565}
       Scale.X {type=MovieSceneScriptingDoubleChannel, default=1}
       Scale.Y {type=MovieSceneScriptingDoubleChannel, default=1}
       Scale.Z {type=MovieSceneScriptingDoubleChannel, default=1}
      ]
     }]
   }
  ]
 }
 BP_NightNight_Generic_Rift_Exit_Long {
  id=#3
  possessed_class=BP_NightNight_Generic_RiftExit_Long_C
  tracks [
   {
    type=MovieSceneSpawnTrack
    sections [{
      type=MovieSceneSpawnSection
      end_frame=2599
      start_frame=1533
      channels [None {type=MovieSceneScriptingBoolChannel, default=false, keys [[1533,false],[1798.8,true],[2599,false]]}]
     }]
   }
   {
    type=MovieScene3DTransformTrack
    sections [{
      type=MovieScene3DTransformSection
      channels [
       Location.X {type=MovieSceneScriptingDoubleChannel, default=120682.3984}
       Location.Y {type=MovieSceneScriptingDoubleChannel, default=-85738.1094}
       Location.Z {type=MovieSceneScriptingDoubleChannel, default=101548}
       Rotation.X {type=MovieSceneScriptingDoubleChannel, default=0}
       Rotation.Y {type=MovieSceneScriptingDoubleChannel, default=-90}
       Rotation.Z {type=MovieSceneScriptingDoubleChannel, default=0}
       Scale.X {type=MovieSceneScriptingDoubleChannel, default=1}
       Scale.Y {type=MovieSceneScriptingDoubleChannel, default=1}
       Scale.Z {type=MovieSceneScriptingDoubleChannel, default=1}
      ]
     }]
   }
  ]
 }
 RiftExit {
  id=#20
  parent_id=#12
  possessed_class=ParticleSystemComponent
  tracks [{
    type=MovieSceneParticleTrack
    display_name=Particle System
    sections [{type=MovieSceneParticleSection, channels [None {type=MovieSceneScriptingParticleChannel, const ["<ParticleKey.TRIGGER: 2>", 1]}]}]
   }]
 }
 BP_NightNight_Generic_Rift_Exit_Long {
  id=#6
  possessed_class=BP_NightNight_Generic_RiftExit_Long_C
  tracks [
   {
    type=MovieSceneSpawnTrack
    sections [{
      type=MovieSceneSpawnSection
      end_frame=2419
      start_frame=1423
      channels [None {type=MovieSceneScriptingBoolChannel, default=false, keys [[1423,false],[1618.8,true],[2419,false]]}]
     }]
   }
   {
    type=MovieScene3DTransformTrack
    sections [{
      type=MovieScene3DTransformSection
      channels [
       Location.X {type=MovieSceneScriptingDoubleChannel, default=-28878.002}
       Location.Y {type=MovieSceneScriptingDoubleChannel, default=58285.9922}
       Location.Z {type=MovieSceneScriptingDoubleChannel, default=101506}
       Rotation.X {type=MovieSceneScriptingDoubleChannel, default=93.6308}
       Rotation.Y {type=MovieSceneScriptingDoubleChannel, default=-90}
       Rotation.Z {type=MovieSceneScriptingDoubleChannel, default=14.0363}
       Scale.X {type=MovieSceneScriptingDoubleChannel, default=1}
       Scale.Y {type=MovieSceneScriptingDoubleChannel, default=1}
       Scale.Z {type=MovieSceneScriptingDoubleChannel, default=1}
      ]
     }]
   }
  ]
 }
]
display_rate {denominator=1, numerator=60}
tick_resolution {denominator=1, numerator=24000}
tracks []
