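# Level Sequence for part three of the NightNight prototype animation, combining three large static mesh actors with a set of nested sequences.
# Two "Start" actors and "Sphere (2)" each hold a fixed transform and a spawn track, and the sphere also has a visibility track.
# A Subsequences track layers seven overlapping sections with evenly staggered starts, all but the first ending on the same frame.
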LevelSequence /Figment_S10/Athena/Prototype/Blueprints/NightNight/NightNight_Part3_Animation
bindings [
 Start {
  id=#1
  possessed_class=FortStaticMeshActor
  tracks [
   {
    type=MovieScene3DTransformTrack
    sections [{
      type=MovieScene3DTransformSection
      channels [
       Location.X {type=MovieSceneScriptingDoubleChannel, default=100796.5625}
       Location.Y {type=MovieSceneScriptingDoubleChannel, default=-7551.8262}
       Location.Z {type=MovieSceneScriptingDoubleChannel, default=33994.5156}
       Rotation.X {type=MovieSceneScriptingDoubleChannel, default=0}
       Rotation.Y {type=MovieSceneScriptingDoubleChannel, default=0}
       Rotation.Z {type=MovieSceneScriptingDoubleChannel, default=0}
       Scale.X {type=MovieSceneScriptingDoubleChannel, default=55}
       Scale.Y {type=MovieSceneScriptingDoubleChannel, default=55}
       Scale.Z {type=MovieSceneScriptingDoubleChannel, default=55}
      ]
     }]
   }
   {type=MovieSceneSpawnTrack, sections [{type=MovieSceneSpawnSection, channels [None {type=MovieSceneScriptingBoolChannel, default=true}]}]}
  ]
 }
 Start {
  id=#2
  possessed_class=FortStaticMeshActor
  tracks [
   {
    type=MovieScene3DTransformTrack
    sections [{
      type=MovieScene3DTransformSection
      channels [
       Location.X {type=MovieSceneScriptingDoubleChannel, default=52512.5156}
       Location.Y {type=MovieSceneScriptingDoubleChannel, default=-31008.4688}
       Location.Z {type=MovieSceneScriptingDoubleChannel, default=14942.8594}
       Rotation.X {type=MovieSceneScriptingDoubleChannel, default=0}
       Rotation.Y {type=MovieSceneScriptingDoubleChannel, default=0}
       Rotation.Z {type=MovieSceneScriptingDoubleChannel, default=0}
       Scale.X {type=MovieSceneScriptingDoubleChannel, default=55}
       Scale.Y {type=MovieSceneScriptingDoubleChannel, default=55}
       Scale.Z {type=MovieSceneScriptingDoubleChannel, default=55}
      ]
     }]
   }
   {type=MovieSceneSpawnTrack, sections [{type=MovieSceneSpawnSection, channels [None {type=MovieSceneScriptingBoolChannel, default=true}]}]}
  ]
 }
 "Sphere (2)" {
  id=#3
  possessed_class=FortStaticMeshActor
  tracks [
   {
    type=MovieScene3DTransformTrack
    sections [{
      type=MovieScene3DTransformSection
      channels [
       Location.X {type=MovieSceneScriptingDoubleChannel, default=-27302.1426}
       Location.Y {type=MovieSceneScriptingDoubleChannel, default=-16645.1055}
       Location.Z {type=MovieSceneScriptingDoubleChannel, default=39756.7305}
       Rotation.X {type=MovieSceneScriptingDoubleChannel, default=0}
       Rotation.Y {type=MovieSceneScriptingDoubleChannel, default=0}
       Rotation.Z {type=MovieSceneScriptingDoubleChannel, default=0}
       Scale.X {type=MovieSceneScriptingDoubleChannel, default=111}
       Scale.Y {type=MovieSceneScriptingDoubleChannel, default=111}
       Scale.Z {type=MovieSceneScriptingDoubleChannel, default=111}
      ]
     }]
   }
   {type=MovieSceneSpawnTrack, sections [{type=MovieSceneSpawnSection, channels [None {type=MovieSceneScriptingBoolChannel, default=true}]}]}
   {
    type=MovieSceneVisibilityTrack
    sections [{type=MovieSceneVisibilitySection, channels [None {type=MovieSceneScriptingBoolChannel, default=true, const [true,1]}]}]
   }
  ]
 }
]
display_rate {denominator=1, numerator=60}
tick_resolution {denominator=1, numerator=24000}
tracks [{
  type=MovieSceneSubTrack
  display_name=Subsequences
  sections [
   {type=MovieSceneSubSection, end_frame=2826, start_frame=1023}
   {type=MovieSceneSubSection, end_frame=3012, start_frame=1053}
   {type=MovieSceneSubSection, end_frame=3012, start_frame=1083}
   {type=MovieSceneSubSection, end_frame=3012, start_frame=1113}
   {type=MovieSceneSubSection, end_frame=3012, start_frame=1143}
   {type=MovieSceneSubSection, end_frame=3012, start_frame=1173}
   {type=MovieSceneSubSection, end_frame=3012, start_frame=1203}
  ]
 }]
