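# Level Sequence for part 4 of the NightNight prototype event, animating a meteor-like "Thing" that shakes loose and departs.
# Frozen crack particles switch to unfreezing ones, dense position keys jitter the Thing, then glitch mesh overlays swell and fade.
# A Camera Rig Rail position track then advances while the attached surround moves away shrinking, the trail collapses and spawn tracks remove the actors.
# Staggered, overlapping subsequences play during the lead-up.
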LevelSequence /Figment_S10/Athena/Prototype/Blueprints/NightNight/NightNight_Part4_Animation
bindings [
 Sphere {
  id=#1
  possessed_class=FortStaticMeshActor
  tracks [
   {
    type=MovieScene3DTransformTrack
    sections [{
      type=MovieScene3DTransformSection
      channels [
       Location.X {type=MovieSceneScriptingDoubleChannel, default=0}
       Location.Y {type=MovieSceneScriptingDoubleChannel, default=0}
       Location.Z {type=MovieSceneScriptingDoubleChannel, default=0}
       Rotation.X {type=MovieSceneScriptingDoubleChannel, default=0}
       Rotation.Y {type=MovieSceneScriptingDoubleChannel, default=0}
       Rotation.Z {type=MovieSceneScriptingDoubleChannel, default=-15.0013}
       Scale.X {type=MovieSceneScriptingDoubleChannel, default=1}
       Scale.Y {type=MovieSceneScriptingDoubleChannel, default=1}
       Scale.Z {type=MovieSceneScriptingDoubleChannel, default=1}
      ]
     }]
   }
   {type=MovieScene3DAttachTrack, sections [{type=MovieScene3DAttachSection, end_frame=3300, start_frame=1023}]}
   {
    type=MovieSceneSpawnTrack
    sections [{type=MovieSceneSpawnSection, channels [None {type=MovieSceneScriptingBoolChannel, default=false, keys [[1023,true],[3142,false]]}]}]
   }
  ]
 }
 MeteorTrailGlitch {
  id=#2
  parent_id=#3
  possessed_class=StaticMeshComponent
  tracks [
   {
    type=MovieScene3DTransformTrack
    sections [{
      type=MovieScene3DTransformSection
      channels [
       Scale.X {type=MovieSceneScriptingDoubleChannel, default=1.7, keys [[2752,1],[2802,1.7]]}
       Scale.Y {type=MovieSceneScriptingDoubleChannel, default=1, const [1,2]}
       Scale.Z {type=MovieSceneScriptingDoubleChannel, default=1.7, keys [[2752,1],[2802,1.7]]}
      ]
     }]
   }
   {
    type=MovieSceneComponentMaterialTrack
    display_name=Material Element 0
    sections [{
      type=MovieSceneParameterSection
      end_frame=3095
      start_frame=2711
      channels [Opacity {type=MovieSceneScriptingFloatChannel, default=1.5, keys [[2750,1],[2752,1.5],[2802,0]]}]
     }]
   }
  ]
 }
 TrailParent {
  id=#4
  parent_id=#3
  possessed_class=SceneComponent
  tracks [{
    type=MovieScene3DTransformTrack
    sections [{
      type=MovieScene3DTransformSection
      channels [
       Scale.X {type=MovieSceneScriptingDoubleChannel, default=0, keys [[2985,1],[3061,0]]}
       Scale.Y {type=MovieSceneScriptingDoubleChannel, default=1.1, keys [[2917,1],[2985,1.1],[3061,1]]}
       Scale.Z {type=MovieSceneScriptingDoubleChannel, default=0, keys [[2985,1],[3061,0]]}
      ]
     }]
   }]
 }
 BP Night Night Thing Moving {
  id=#3
  possessed_class=BP_NightNight_Thing_Moving_C
  tracks [
   {
    type=MovieSceneSpawnTrack
    sections [{
      type=MovieSceneSpawnSection
      channels [None {type=MovieSceneScriptingBoolChannel, default=false, keys [[1023,true],[3093,false],[3846.9975,false]]}]
     }]
   }
   {
    type=MovieScene3DTransformTrack
    sections [{
      type=MovieScene3DTransformSection
      channels [
       Location.X {
        type=MovieSceneScriptingDoubleChannel
        default=-579.2128
        thinned_from=114
        keys [[1023.0025,0],[2532.065,-0.9001],[2534.81,24.4326],[2537.5575,-0.9001],[2540.3025,24.4326],[2543.0475,-0.9001],[2545.7975,24.4326],[2548.545,-0.9001],[2551.2875,24.4326],[2554.0375,-0.9001],[2556.7875,24.4326],[2559.53,-0.9001],[2562.2775,51.1362],[2565.0225,-0.9001],[2567.7725,51.1362],[2570.52,-0.9001],[2573.2625,51.1362],[2576.0125,-0.9001],[2579.2175,-0.9001],[2581.9625,51.1362],[2584.71,-0.9001],[2587.455,71.7821],[2590.2,-0.9001],[2592.95,71.7821],[2595.6975,-0.9001],[2598.44,80.9813],[2601.19,-0.9001],[2603.94,80.9813],[2606.6825,-0.9001],[2609.43,80.9813],[2612.175,-0.9001],[2614.925,65.6493],[2617.6725,-0.9001],[2620.415,65.6493],[2623.165,-0.9001],[2625.91,45.3967],[2628.3575,-0.9001],[2629.9,45.3967],[2631.4425,-0.9001],[2632.985,45.3967],[2634.525,-0.9001],[2636.575,45.3967],[2638.6275,-0.9001],[2640.68,45.3967],[2642.73,-0.9001],[2644.78,45.3967],[2646.8325,-0.9001],[2648.8825,45.3967],[2650.9325,-0.9001],[2652.9825,45.3967],[2655.2,-0.9001],[2656.89,-0.9001],[2658.4325,19.7403],[2659.975,-0.9001],[2661.5175,19.7403],[2663.0575,-0.9001],[2665.1075,19.7403],[2667.16,-0.9001],[2669.2125,19.7403],[2671.2625,-0.9001],[2673.3125,19.7403],[2675.365,-0.9001],[2677.415,19.7403],[2679.465,-0.9001],[2681.515,19.7403],[2683.7325,-0.9001],[2687.69,-0.9001],[2689.2325,19.7403],[2690.775,-0.9001],[2692.3175,19.7403],[2693.8575,-0.9001],[2695.9075,19.7403],[2697.96,-0.9001],[2700.0125,19.7403],[2702.0625,-0.9001],[2704.1125,19.7403],[2706.165,-0.9001],[2708.215,19.7403],[2710.265,-0.9001],[2712.315,19.7403],[2714.5325,-0.9001],[2718.2225,-0.9001],[2720,19.7403],[2722,-0.9001],[2723,19.7403],[2725,-0.9001],[2727,19.7403],[2728,-0.9001],[2729,19.7403],[2730,-0.9001],[2733,19.7403],[2734,-0.9001],[2735,19.7403],[2737,-0.9001],[2738,19.7403],[2739,-0.9001],[2744,-0.9001],[2744,19.7403],[2745,19.7403],[2747,-0.9001],[2750,16.3813],[2753,-0.9001],[2754,16.3813],[2757,-0.9001],[2759,16.3813],[2760,-0.9001],[2762,16.3813],[2765,-0.9001],[2767,16.3813],[2768,-0.9001]]
       }
       Location.Y {
        type=MovieSceneScriptingDoubleChannel
        default=-22.5073
        thinned_from=119
        keys [[1023.0025,0],[2532.065,-0.9001],[2534.81,24.4326],[2537.5575,-0.9001],[2540.3025,24.4326],[2543.0475,-0.9001],[2545.7975,24.4326],[2548.545,-0.9001],[2551.2875,24.4326],[2554.0375,-0.9001],[2556.7875,24.4326],[2559.53,-0.9001],[2562.2775,51.1362],[2565.0225,-0.9001],[2567.7725,51.1362],[2570.52,-0.9001],[2573.2625,51.1362],[2576.0125,-0.9001],[2579.2175,-0.9001],[2581.9625,51.1362],[2584.71,-0.9001],[2587.455,71.7821],[2590.2,-0.9001],[2592.95,71.7821],[2595.6975,-0.9001],[2598.44,80.9813],[2601.19,-0.9001],[2603.94,80.9813],[2606.6825,-0.9001],[2609.43,80.9813],[2612.175,-0.9001],[2614.925,65.6493],[2617.6725,-0.9001],[2620.415,65.6493],[2623.165,-0.9001],[2625.91,45.3967],[2628.3575,-0.9001],[2629.9,45.3967],[2631.4425,-0.9001],[2632.985,45.3967],[2634.525,-0.9001],[2636.575,45.3967],[2638.6275,-0.9001],[2640.68,45.3967],[2642.73,-0.9001],[2644.78,45.3967],[2646.8325,-0.9001],[2648.8825,45.3967],[2650.9325,-0.9001],[2652.9825,45.3967],[2655.2,-0.9001],[2656.89,-0.9001],[2658.4325,19.7403],[2659.975,-0.9001],[2661.5175,19.7403],[2663.0575,-0.9001],[2665.1075,19.7403],[2667.16,-0.9001],[2669.2125,19.7403],[2671.2625,-0.9001],[2673.3125,19.7403],[2675.365,-0.9001],[2677.415,19.7403],[2679.465,-0.9001],[2681.515,19.7403],[2683.7325,-0.9001],[2687.69,-0.9001],[2689.2325,19.7403],[2690.775,-0.9001],[2692.3175,19.7403],[2693.8575,-0.9001],[2695.9075,19.7403],[2697.96,-0.9001],[2700.0125,19.7403],[2702.0625,-0.9001],[2704.1125,19.7403],[2706.165,-0.9001],[2708.215,19.7403],[2710.265,-0.9001],[2712.315,19.7403],[2714.5325,-0.9001],[2718.2225,-0.9001],[2720,19.7403],[2722,-0.9001],[2723,19.7403],[2725,-0.9001],[2727,19.7403],[2728,-0.9001],[2729,19.7403],[2730,-0.9001],[2733,19.7403],[2734,-0.9001],[2735,19.7403],[2737,-0.9001],[2738,19.7403],[2739,-0.9001],[2744,-0.9001],[2744,19.7403],[2745,19.7403],[2747,-0.9001],[2750,16.3813],[2753,-0.9001],[2754,16.3813],[2757,-0.9001],[2759,16.3813],[2760,-0.9001],[2762,16.3813],[2765,-0.9001],[2767,16.3813],[2768,-0.9001],[2773,-0.9001],[2788,-9.2404],[2820,-0.9001],[2859,-2.6856],[2915,-1.4133]]
       }
       Location.Z {
        type=MovieSceneScriptingDoubleChannel
        default=255
        keys [[1023,0],[2523.825,36.0551],[2526.5725,0],[2529.315,36.0551],[2532.065,0],[2534.3325,0],[2537.0775,36.0551],[2539.825,0],[2542.57,36.0551],[2545.315,0],[2548.065,36.0551],[2550.8125,0],[2553.555,36.0551],[2556.305,0],[2559.055,67.742],[2561.7975,0],[2564.545,67.742],[2567.29,0],[2570.04,67.742],[2572.7875,0],[2575.53,67.742],[2578.28,8.2661],[2581.485,0],[2584.23,110.4503],[2586.9775,0],[2589.7225,110.4503],[2592.4675,0],[2595.2175,110.4503],[2597.965,0],[2600.7075,110.4503],[2603.4575,0],[2606.2075,200],[2608.95,0],[2611.6975,200],[2614.4425,0],[2617.1925,200],[2619.94,0],[2622.6825,200],[2625.4325,0],[2628.1775,166.9355],[2630.625,0],[2632.1675,166.9355],[2633.71,0],[2635.2525,166.9355],[2636.7925,0],[2638.8425,166.9355],[2640.895,0],[2642.9475,129.7379],[2644.9975,0],[2647.0475,129.7379],[2649.1,0],[2651.15,129.7379],[2653.2,0],[2655.25,129.7379],[2657.4675,0],[2659.1575,0],[2660.7,92.5403],[2662.2425,0],[2663.785,92.5403],[2665.325,0],[2667.375,92.5403],[2669.4275,0],[2671.48,92.5403],[2673.53,0],[2675.58,92.5403],[2677.6325,0],[2679.6825,92.5403],[2681.7325,0],[2683.7825,92.5403],[2686,0],[2688,166.9355],[2689.54,0],[2691.59,166.9355],[2693.6425,0],[2695.695,129.7379],[2697.745,0],[2699.795,129.7379],[2701.8475,0],[2703.8975,129.7379],[2705.9475,0],[2707.9975,129.7379],[2710.215,0],[2711.905,0],[2713.4475,92.5403],[2714.99,0],[2716.5325,92.5403],[2718.0725,0],[2720,92.5403],[2722,0],[2725,92.5403],[2727,0],[2728,92.5403],[2729,0],[2730,92.5403],[2733,0],[2734,92.5403],[2735,0],[2737,166.9355],[2738,0],[2739,166.9355],[2740,0],[2743,129.7379],[2744,0],[2745,129.7379],[2748,-0.1465],[2750,108.4779],[2753,-0.1465],[2756,108.4779],[2757,-0.1465],[2759,-0.1465],[2760,77.3338],[2762,-0.1465],[2763,77.3338],[2765,-0.1465],[2767,77.3338],[2768,-0.1465],[2773,-0.1465],[2788,19.9124],[2831,0],[2860,-0.8906],[2912,-0.7769]]
       }
      ]
     }]
   }
   {type=MovieScene3DAttachTrack, sections [{type=MovieScene3DAttachSection, end_frame=3846, start_frame=1023}]}
  ]
 }
 Camera Rig Rail {
  id=#5
  possessed_class=CameraRig_Rail
  tracks [
   {type=MovieSceneSpawnTrack, sections [{type=MovieSceneSpawnSection, channels [None {type=MovieSceneScriptingBoolChannel, default=true}]}]}
   {
    type=MovieScene3DTransformTrack
    sections [{
      type=MovieScene3DTransformSection
      channels [
       Location.X {type=MovieSceneScriptingDoubleChannel, default=16150.1133}
       Location.Y {type=MovieSceneScriptingDoubleChannel, default=7339.5381}
       Location.Z {type=MovieSceneScriptingDoubleChannel, default=9408}
       Rotation.X {type=MovieSceneScriptingDoubleChannel, default=0}
       Rotation.Y {type=MovieSceneScriptingDoubleChannel, default=0}
       Rotation.Z {type=MovieSceneScriptingDoubleChannel, default=-15.2944}
       Scale.X {type=MovieSceneScriptingDoubleChannel, default=1}
       Scale.Y {type=MovieSceneScriptingDoubleChannel, default=1}
       Scale.Z {type=MovieSceneScriptingDoubleChannel, default=1}
      ]
     }]
   }
   {
    type=MovieSceneFloatTrack
    display_name=Current Position on Rail
    sections [{
      type=MovieSceneFloatSection
      channels [None {type=MovieSceneScriptingFloatChannel, default=1, keys [[1023,0],[2742,0],[2768,0.0103],[2873,0.0166],[3072,1.0064],[3142,1]]}]
     }]
   }
  ]
 }
 MeteorGlitch {
  id=#6
  parent_id=#3
  possessed_class=StaticMeshComponent
  tracks [
   {
    type=MovieSceneComponentMaterialTrack
    display_name=Material Element 0
    sections [{
      type=MovieSceneParameterSection
      end_frame=3095
      start_frame=2711
      channels [Opacity {type=MovieSceneScriptingFloatChannel, default=2, keys [[2750,1],[2752,2],[2802,0]]}]
     }]
   }
   {
    type=MovieScene3DTransformTrack
    sections [{
      type=MovieScene3DTransformSection
      channels [
       Scale.X {type=MovieSceneScriptingDoubleChannel, default=1.5, keys [[2752,1],[2802,1.5]]}
       Scale.Y {type=MovieSceneScriptingDoubleChannel, default=1.5, keys [[2752,1],[2802,1.5]]}
       Scale.Z {type=MovieSceneScriptingDoubleChannel, default=1.5, keys [[2752,1],[2802,1.5]]}
      ]
     }]
   }
  ]
 }
 P_Geode_CrackExit_Unfreezing {
  id=#7
  parent_id=#3
  possessed_class=ParticleSystemComponent
  tracks [{
    type=MovieSceneParticleTrack
    display_name=Particle System
    sections [{
      type=MovieSceneParticleSection
      channels [None {
        type=MovieSceneScriptingParticleChannel
        keys [[1023, "<ParticleKey.DEACTIVATE: 1>"], [2475, "<ParticleKey.ACTIVATE: 0>"], [2943, "<ParticleKey.DEACTIVATE: 1>"]]
       }]
     }]
   }]
 }
 BP_NightNight_ThingSurround_Moving {
  id=#8
  possessed_class=BP_NightNight_ThingSurround_Moving_C
  tracks [
   {
    type=MovieSceneSpawnTrack
    sections [{type=MovieSceneSpawnSection, channels [None {type=MovieSceneScriptingBoolChannel, default=false, keys [[1023,true],[3045,false]]}]}]
   }
   {
    type=MovieScene3DTransformTrack
    sections [{
      type=MovieScene3DTransformSection
      channels [
       Location.X {type=MovieSceneScriptingDoubleChannel, default=0, keys [[2752,0],[3015,33427.6875]]}
       Location.Y {type=MovieSceneScriptingDoubleChannel, default=0, keys [[2752,0],[3015,-6949.9727]]}
       Location.Z {type=MovieSceneScriptingDoubleChannel, default=0, keys [[2752,0],[3015,-17140.9883]]}
       Scale.X {type=MovieSceneScriptingDoubleChannel, default=0.25, keys [[1023,1],[2752,1],[3015,0]]}
       Scale.Y {type=MovieSceneScriptingDoubleChannel, default=0.25, keys [[1023,1],[2752,1],[3015,0]]}
       Scale.Z {type=MovieSceneScriptingDoubleChannel, default=0.25, keys [[1023,1],[2752,1],[3015,0]]}
      ]
     }]
   }
   {type=MovieScene3DAttachTrack, sections [{type=MovieScene3DAttachSection, end_frame=3300, start_frame=1023}]}
  ]
 }
 P_Geode_CrackExit_Frozen {
  id=#9
  parent_id=#3
  possessed_class=ParticleSystemComponent
  tracks [{
    type=MovieSceneParticleTrack
    display_name=Particle System
    sections [{
      type=MovieSceneParticleSection
      channels [None {type=MovieSceneScriptingParticleChannel, keys [[1023, "<ParticleKey.ACTIVATE: 0>"], [2475, "<ParticleKey.DEACTIVATE: 1>"]]}]
     }]
   }]
 }
]
display_rate {denominator=1, numerator=60}
tick_resolution {denominator=1, numerator=24000}
tracks [{
  type=MovieSceneSubTrack
  display_name=Subsequences
  sections [
   {type=MovieSceneSubSection, end_frame=1859, start_frame=1443}
   {type=MovieSceneSubSection, end_frame=1948, start_frame=1503}
   {type=MovieSceneSubSection, end_frame=2008, start_frame=1563}
   {type=MovieSceneSubSection, end_frame=2068, start_frame=1623}
   {type=MovieSceneSubSection, end_frame=2128, start_frame=1683}
   {type=MovieSceneSubSection, end_frame=2158, start_frame=1713}
   {type=MovieSceneSubSection, end_frame=2188, start_frame=1743}
   {type=MovieSceneSubSection, end_frame=2218, start_frame=1773}
   {type=MovieSceneSubSection, end_frame=2248, start_frame=1803}
   {type=MovieSceneSubSection, end_frame=2278, start_frame=1833}
   {type=MovieSceneSubSection, end_frame=2308, start_frame=1863}
   {type=MovieSceneSubSection, end_frame=2338, start_frame=1893}
   {type=MovieSceneSubSection, end_frame=2368, start_frame=1923}
   {type=MovieSceneSubSection, end_frame=2398, start_frame=1953}
   {type=MovieSceneSubSection, end_frame=2428, start_frame=1983}
   {type=MovieSceneSubSection, end_frame=2458, start_frame=2013}
   {type=MovieSceneSubSection, end_frame=2488, start_frame=2043}
   {type=MovieSceneSubSection, end_frame=2518, start_frame=2073}
   {type=MovieSceneSubSection, end_frame=1739, start_frame=1023}
   {type=MovieSceneSubSection, end_frame=1833, start_frame=1262}
   {type=MovieSceneSubSection, end_frame=1859, start_frame=1382}
  ]
 }]
