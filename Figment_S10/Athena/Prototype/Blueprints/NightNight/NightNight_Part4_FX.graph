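# Level Sequence for the FX of part 4 of the NightNight event, staging a meteor breaking open and entering a rift.
# Spawn and particle tracks time emitters for rift enter and exit loops, crack shards, rocket streaks, shell break, meteor wake-up and the rift eating the meteor.
# Transform tracks place several effects, scale the crack shards and sweep the wake-up emitter along the meteor's path.
# An MPC_Nexus track animates repeating, accelerating meteor wave pulses, a continuous wave and glow intensity.
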LevelSequence /Figment_S10/Athena/Prototype/Blueprints/NightNight/NightNight_Part4_FX
bindings [
 P_NightNight_CrackEnter_ShardsOnly {
  id=#1
  possessed_class=Emitter
  tracks [
   {
    type=MovieSceneSpawnTrack
    sections [{
      type=MovieSceneSpawnSection
      channels [None {type=MovieSceneScriptingBoolChannel, default=false, keys [[0,false],[259,true],[1061.9987,false]]}]
     }]
   }
   {
    type=MovieScene3DTransformTrack
    sections [{
      type=MovieScene3DTransformSection
      channels [
       Location.X {type=MovieSceneScriptingDoubleChannel, default=25049.2285}
       Location.Y {type=MovieSceneScriptingDoubleChannel, default=1895.4473}
       Location.Z {type=MovieSceneScriptingDoubleChannel, default=3704.5798}
       Scale.X {type=MovieSceneScriptingDoubleChannel, default=1.2, keys [[259,1],[599,1.2],[757,0.9]]}
       Scale.Y {type=MovieSceneScriptingDoubleChannel, default=1.2, keys [[259,1],[599,1.2],[757,0.8]]}
       Scale.Z {type=MovieSceneScriptingDoubleChannel, default=0.9, keys [[259,1],[599,1.2],[757,0.9]]}
      ]
     }]
   }
   {
    type=MovieSceneParticleTrack
    display_name=Particle System
    sections [{type=MovieSceneParticleSection, channels [None {type=MovieSceneScriptingParticleChannel, const ["<ParticleKey.TRIGGER: 2>", 31]}]}]
   }
  ]
 }
 P_MeteorRiftEnterLoop {
  id=#2
  possessed_class=Emitter
  tracks [
   {
    type=MovieSceneSpawnTrack
    sections [{type=MovieSceneSpawnSection, channels [None {type=MovieSceneScriptingBoolChannel, default=false, keys [[0,false],[148,true],[814,false]]}]}]
   }
   {
    type=MovieSceneParticleTrack
    display_name=Particle System
    sections [{
      type=MovieSceneParticleSection
      channels [None {type=MovieSceneScriptingParticleChannel, keys [[148, "<ParticleKey.ACTIVATE: 0>"], [734, "<ParticleKey.DEACTIVATE: 1>"]]}]
     }]
   }
  ]
 }
 P_NightNight_MeteorShellBreak {
  id=#3
  possessed_class=Emitter
  tracks [
   {
    type=MovieSceneSpawnTrack
    sections [{
      type=MovieSceneSpawnSection
      channels [None {type=MovieSceneScriptingBoolChannel, default=false, keys [[1,false],[861,true],[1061.9987,false]]}]
     }]
   }
   {
    type=MovieScene3DTransformTrack
    sections [{
      type=MovieScene3DTransformSection
      channels [
       Location.X {type=MovieSceneScriptingDoubleChannel, default=25071.8008}
       Location.Y {type=MovieSceneScriptingDoubleChannel, default=1936.712}
       Location.Z {type=MovieSceneScriptingDoubleChannel, default=3778.0439}
       Rotation.X {type=MovieSceneScriptingDoubleChannel, default=-28.5}
       Rotation.Y {type=MovieSceneScriptingDoubleChannel, default=26.4821}
       Rotation.Z {type=MovieSceneScriptingDoubleChannel, default=-210}
       Scale.X {type=MovieSceneScriptingDoubleChannel, default=1}
       Scale.Y {type=MovieSceneScriptingDoubleChannel, default=1}
       Scale.Z {type=MovieSceneScriptingDoubleChannel, default=1}
      ]
     }]
   }
   {
    type=MovieSceneParticleTrack
    display_name=Particle System
    sections [{type=MovieSceneParticleSection, channels [None {type=MovieSceneScriptingParticleChannel, const ["<ParticleKey.TRIGGER: 2>", 1]}]}]
   }
  ]
 }
 Rift Eats Meteor {
  id=#4
  possessed_class=Emitter
  tracks [
   {
    type=MovieSceneSpawnTrack
    sections [{type=MovieSceneSpawnSection, channels [None {type=MovieSceneScriptingBoolChannel, default=false, keys [[0,false],[959,true],[1069,false]]}]}]
   }
   {
    type=MovieSceneParticleTrack
    display_name=Particle System
    sections [{
      type=MovieSceneParticleSection
      channels [None {type=MovieSceneScriptingParticleChannel, keys [[959, "<ParticleKey.ACTIVATE: 0>"], [1034.0013, "<ParticleKey.DEACTIVATE: 1>"]]}]
     }]
   }
  ]
 }
 P_MeteorRiftEnterLoop Growing {
  id=#5
  possessed_class=Emitter
  tracks [
   {
    type=MovieSceneSpawnTrack
    sections [{type=MovieSceneSpawnSection, channels [None {type=MovieSceneScriptingBoolChannel, default=false, keys [[0,false],[734,true],[1114,false]]}]}]
   }
   {
    type=MovieSceneParticleTrack
    display_name=Particle System
    sections [{
      type=MovieSceneParticleSection
      channels [None {type=MovieSceneScriptingParticleChannel, keys [[734, "<ParticleKey.ACTIVATE: 0>"], [1034.0013, "<ParticleKey.DEACTIVATE: 1>"]]}]
     }]
   }
  ]
 }
 P_NightNight_RocketStreaks {
  id=#6
  possessed_class=Emitter
  tracks [
   {
    type=MovieSceneSpawnTrack
    sections [{
      type=MovieSceneSpawnSection
      channels [None {type=MovieSceneScriptingBoolChannel, default=false, keys [[1,false],[524,true],[1061.9987,false]]}]
     }]
   }
   {
    type=MovieScene3DTransformTrack
    sections [{
      type=MovieScene3DTransformSection
      channels [
       Location.X {type=MovieSceneScriptingDoubleChannel, default=25071.8008}
       Location.Y {type=MovieSceneScriptingDoubleChannel, default=1936.712}
       Location.Z {type=MovieSceneScriptingDoubleChannel, default=3778.0439}
       Rotation.X {type=MovieSceneScriptingDoubleChannel, default=-28.5}
       Rotation.Y {type=MovieSceneScriptingDoubleChannel, default=26.4821}
       Rotation.Z {type=MovieSceneScriptingDoubleChannel, default=-210}
       Scale.X {type=MovieSceneScriptingDoubleChannel, default=1}
       Scale.Y {type=MovieSceneScriptingDoubleChannel, default=1}
       Scale.Z {type=MovieSceneScriptingDoubleChannel, default=1}
      ]
     }]
   }
   {
    type=MovieSceneParticleTrack
    display_name=Particle System
    sections [{type=MovieSceneParticleSection, channels [None {type=MovieSceneScriptingParticleChannel, const ["<ParticleKey.TRIGGER: 2>", 1]}]}]
   }
  ]
 }
 P_NightNight_Meteor_WakeUp {
  id=#7
  possessed_class=Emitter
  tracks [
   {
    type=MovieSceneSpawnTrack
    sections [{type=MovieSceneSpawnSection, channels [None {type=MovieSceneScriptingBoolChannel, default=false, keys [[0,false],[861,true],[991,false]]}]}]
   }
   {
    type=MovieScene3DTransformTrack
    sections [{
      type=MovieScene3DTransformSection
      channels [
       Location.X {type=MovieSceneScriptingDoubleChannel, default=16677.1367, keys [[861,16677.1367],[905,17792.2344],[934,18446.75],[970,26700.916]]}
       Location.Y {type=MovieSceneScriptingDoubleChannel, default=7074.7451, keys [[861,7074.7451],[905,6440.7197],[934,6068.5742],[970,1375.4087]]}
       Location.Z {type=MovieSceneScriptingDoubleChannel, default=9455.4717, keys [[861,9455.4717],[905,8733.4268],[934,8309.6172],[970,2964.918]]}
       Rotation.X {type=MovieSceneScriptingDoubleChannel, default=-172.8912}
       Rotation.Y {type=MovieSceneScriptingDoubleChannel, default=-23.9389}
       Rotation.Z {type=MovieSceneScriptingDoubleChannel, default=-28.8447}
       Scale.X {type=MovieSceneScriptingDoubleChannel, default=1}
       Scale.Y {type=MovieSceneScriptingDoubleChannel, default=1}
       Scale.Z {type=MovieSceneScriptingDoubleChannel, default=1}
      ]
     }]
   }
   {
    type=MovieSceneParticleTrack
    display_name=Particle System
    sections [{
      type=MovieSceneParticleSection
      channels [None {type=MovieSceneScriptingParticleChannel, keys [[861, "<ParticleKey.ACTIVATE: 0>"], [960, "<ParticleKey.DEACTIVATE: 1>"]]}]
     }]
   }
  ]
 }
 P_MeteorRiftExitLoop {
  id=#8
  possessed_class=Emitter
  tracks [
   {
    type=MovieSceneParticleTrack
    display_name=Particle System
    sections [{
      type=MovieSceneParticleSection
      channels [None {type=MovieSceneScriptingParticleChannel, keys [[0, "<ParticleKey.ACTIVATE: 0>"], [790, "<ParticleKey.DEACTIVATE: 1>"]]}]
     }]
   }
   {
    type=MovieSceneSpawnTrack
    sections [{type=MovieSceneSpawnSection, channels [None {type=MovieSceneScriptingBoolChannel, default=true, keys [[0,true],[846.4987,false]]}]}]
   }
   {
    type=MovieScene3DTransformTrack
    sections [{
      type=MovieScene3DTransformSection
      channels [
       Location.X {type=MovieSceneScriptingDoubleChannel, default=-133204.6406}
       Location.Y {type=MovieSceneScriptingDoubleChannel, default=93153.2734}
       Location.Z {type=MovieSceneScriptingDoubleChannel, default=96423.2578}
       Rotation.X {type=MovieSceneScriptingDoubleChannel, default=4.5945}
       Rotation.Y {type=MovieSceneScriptingDoubleChannel, default=-28.4802}
       Rotation.Z {type=MovieSceneScriptingDoubleChannel, default=-30.6982}
       Scale.X {type=MovieSceneScriptingDoubleChannel, default=1}
       Scale.Y {type=MovieSceneScriptingDoubleChannel, default=1}
       Scale.Z {type=MovieSceneScriptingDoubleChannel, default=1}
      ]
     }]
   }
  ]
 }
]
display_rate {denominator=1, numerator=30}
tick_resolution {denominator=1, numerator=24000}
tracks [{
  type=MovieSceneMaterialParameterCollectionTrack
  display_name=MPC_Nexus
  sections [{
    type=MovieSceneParameterSection
    channels [
     MeteorWave1 {
      type=MovieSceneScriptingFloatChannel
      default=1
      keys [[78,0],[150,1],[196,0],[267,1],[277,0],[361,1],[362,0],[418,1],[419,0.2],[477,1],[478,0.25],[524,1],[525,0.25],[562,1]]
     }
     MeteorWave2 {
      type=MovieSceneScriptingFloatChannel
      default=1
      keys [[246,0],[330,1],[333,0.2],[390,1],[391,0.25],[450,1],[451,0.3],[497,1],[499,0.25],[535,1]]
     }
     MeteorWaveContinuous {type=MovieSceneScriptingFloatChannel, default=1, keys [[0,0],[465,0],[599,0.4835],[861,1],[871,0]]}
     MeteorWaveContinuousAdd {type=MovieSceneScriptingFloatChannel, default=4.7979, keys [[480,0],[861,4.7979],[987,10]]}
     MeteorGlowIntesnify {type=MovieSceneScriptingFloatChannel, default=0.3, keys [[-2510,1],[744,1],[757,0.7813],[861,0.5957],[871,0.2944],[913,0.4672],[987,0.5215]]}
    ]
   }]
 }]
